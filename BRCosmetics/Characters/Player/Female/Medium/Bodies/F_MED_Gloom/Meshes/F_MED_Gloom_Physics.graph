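# Physics asset for the F_MED_Gloom cosmetic body, covering a squid accessory with dangling chain and tentacles attached at the right thigh.
# Capsule bodies form a tree from dyn_squid_MSTR through dyn_squid_base into three two-bone branches: one chain and two tentacles.
# Squid joints share the same swing and twist limits, while the master-to-thigh_r constraint leaves swing and twist free.
# The thigh capsule is the only body outside the squid bones.
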PhysicsAsset /BRCosmetics/Characters/Player/Female/Medium/Bodies/F_MED_Gloom/Meshes/F_MED_Gloom_Physics
bodies {
 dyn_squid_MSTR [{shapeType=Capsule, length=1, radius=1, shapeName=dyn_squid_MSTR_capsule}]
 dyn_squid_base [{shapeType=Capsule, length=7.8716, radius=2.1284, shapeName=dyn_squid_base_capsule, center=<5.6096 0 0>, rotation=<rot 90 0 0>}]
 dyn_squid_chain_01 [{shapeType=Capsule, length=7.8716, radius=2.1284, shapeName=dyn_squid_chain_01_capsule, center=<5.6096 0 0>, rotation=<rot 90 0 0>}]
 dyn_squid_chain_02 [{shapeType=Capsule, length=7.8716, radius=2.1284, shapeName=dyn_squid_chain_01_capsule, center=<5.6096 0 0>, rotation=<rot 90 0 0>}]
 dyn_squid_tenticle_a_01 [{shapeType=Capsule, length=7.8716, radius=2.1284, shapeName=dyn_squid_tenticle_a_01_capsule, center=<5.6096 0 0>, rotation=<rot 90 0 0>}]
 dyn_squid_tenticle_a_02 [{shapeType=Capsule, length=7.8716, radius=2.1284, shapeName=dyn_squid_tenticle_a_01_capsule, center=<5.6096 0 0>, rotation=<rot 90 0 0>}]
 dyn_squid_tenticle_b_01 [{shapeType=Capsule, length=7.8716, radius=2.1284, shapeName=dyn_squid_tenticle_b_01_capsule, center=<5.6096 0 0>, rotation=<rot 90 0 0>}]
 dyn_squid_tenticle_b_02 [{shapeType=Capsule, length=7.8716, radius=2.1284, shapeName=dyn_squid_tenticle_b_01_capsule, center=<5.6096 0 0>, rotation=<rot 90 0 0>}]
 thigh_r [{shapeType=Capsule, length=30.7417, radius=8.0665, shapeName=thigh_r_capsule, center=<21.4829 0 -0.0085>, rotation=<rot 85.735 0 0.0001>}]
}
constraints [
 {bone1Name=dyn_squid_chain_01, bone2Name=dyn_squid_base, swing1LimitDegrees=60, swing2LimitDegrees=60, twistLimitDegrees=45}
 {bone1Name=dyn_squid_tenticle_b_01, bone2Name=dyn_squid_base, swing1LimitDegrees=60, swing2LimitDegrees=60, twistLimitDegrees=45}
 {bone1Name=dyn_squid_tenticle_a_01, bone2Name=dyn_squid_base, swing1LimitDegrees=60, swing2LimitDegrees=60, twistLimitDegrees=45}
 {bone1Name=dyn_squid_base, bone2Name=dyn_squid_MSTR, swing1LimitDegrees=60, swing2LimitDegrees=60, twistLimitDegrees=45}
 {
  bone1Name=dyn_squid_MSTR
  bone2Name=thigh_r
  swing1LimitDegrees=45
  swing1Motion=Free
  swing2LimitDegrees=45
  swing2Motion=Free
  twistLimitDegrees=45
  twistMotion=Free
 }
 {bone1Name=dyn_squid_tenticle_b_02, bone2Name=dyn_squid_tenticle_b_01, swing1LimitDegrees=60, swing2LimitDegrees=60, twistLimitDegrees=45}
 {bone1Name=dyn_squid_tenticle_a_02, bone2Name=dyn_squid_tenticle_a_01, swing1LimitDegrees=60, swing2LimitDegrees=60, twistLimitDegrees=45}
 {bone1Name=dyn_squid_chain_02, bone2Name=dyn_squid_chain_01, swing1LimitDegrees=60, swing2LimitDegrees=60, twistLimitDegrees=45}
]
