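# Control Rig for the F_MED_Gloom female character's hat part that corrects hat and back-hair bones as the head moves.
# A spherical pose reader on head relative to neck_01 weights a rotation and Z-scale offset on dyn_hat_MSTR.
# A second, parentless reader with debug drawing weights a dyn_back_brim offset and rotation-constrains six back-hair bones to spine_05.
# A shared function applies offsets additively to current local transforms, using Euler angles for rotation.
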ControlRigBlueprint /BRCosmetics/Characters/Player/Female/Medium/Bodies/F_MED_Gloom/Meshes/Parts/F_MED_Gloom_Hat_Crouching_FaceAcc_CtrlRig
graphs [
 ContainedGraph {nodes [
   n1 = RigVMFunctionEntryNode()
   RigUnit_GetTransform = &RigUnit_GetTransform(Space=LocalSpace, bInitial=false, Item=@n1.Joint)
   n2 = &RigVMFunction_MathQuaternionToEuler(RotationOrder=ZYX, Value=@RigUnit_GetTransform.Transform.Rotation)
   > RigUnit_SetTransform = &RigUnit_SetTransform(
    Space=LocalSpace
    Value=(Rotation=(X=0.0132,Y=0.4099,Z=-0.0222,W=0.9118),Translation=(X=23.6045,Y=-6.37,Z=140.1115),Scale3D=(X=1,Y=1,Z=1))
    bInitial=false
    Item=@n1.Joint
    Weight=@n1.Weight
    Value.Translation.Z=&RigVMFunction_MathFloatAdd(B=@n1.Translate Z, A=@RigUnit_GetTransform.Transform.Translation.Z) {template="Add::Execute(in A,in B,out Result)"}.Result
    Value.Scale3D.Z=&RigVMFunction_MathFloatAdd(B=@n1.Scale Z, A=@RigUnit_GetTransform.Transform.Scale3D.Z) {template="Add::Execute(in A,in B,out Result)"}.Result
    Value.Translation.Y=&RigVMFunction_MathFloatAdd(B=@n1.Translate Y, A=@RigUnit_GetTransform.Transform.Translation.Y) {template="Add::Execute(in A,in B,out Result)"}.Result
    Value.Scale3D.Y=&RigVMFunction_MathFloatAdd(B=@n1.Scale Y, A=@RigUnit_GetTransform.Transform.Scale3D.Y) {template="Add::Execute(in A,in B,out Result)"}.Result
    Value.Translation.X=&RigVMFunction_MathFloatAdd(B=@n1.Translate X, A=@RigUnit_GetTransform.Transform.Translation.X) {template="Add::Execute(in A,in B,out Result)"}.Result
    Value.Scale3D.X=&RigVMFunction_MathFloatAdd(B=@n1.Scale X, A=@RigUnit_GetTransform.Transform.Scale3D.X) {template="Add::Execute(in A,in B,out Result)"}.Result
    Value.Rotation=&RigVMFunction_MathQuaternionFromEuler(
     Euler=(X=0,Y=0,Z=0)
     Euler.X=&RigVMFunction_MathFloatAdd(B=@n1.Rotate X, A=@n2.Result.X) {template="Add::Execute(in A,in B,out Result)"}.Result
     Euler.Y=&RigVMFunction_MathFloatAdd(B=@n1.Rotate Y, A=@n2.Result.Y) {template="Add::Execute(in A,in B,out Result)"}.Result
     Euler.Z=&RigVMFunction_MathFloatAdd(B=@n1.Rotate Z, A=@n2.Result.Z) {template="Add::Execute(in A,in B,out Result)"}.Result
    ).Result
   )
   > RigVMFunctionReturnNode()
  ]}
 RigVMFunctionLibrary {nodes [Set Transforms for Bone = RigVMCollapseNode(
    ExecuteContext=()
    Joint=()
    Rotate X=0
    Rotate Y=0
    Rotate Z=0
    Scale X=0
    Scale Y=0
    Scale Z=0
    Translate X=0
    Translate Y=0
    Translate Z=0
    Weight=1
   )]}
 RigVMModel {nodes [
   RigUnit_BeginExecution = &RigUnit_BeginExecution()
   n1 = &RigUnit_SphericalPoseReader(
    ActiveRegionScaleFactors=(PositiveWidth=1,NegativeWidth=1,PositiveHeight=1,NegativeHeight=1)
    ActiveRegionSize=0.2
    Debug=(bDrawDebug=false,DebugScale=25,DebugSegments=20,DebugThickness=0.25)
    DriverAxis=(X=1,Y=0,Z=0)
    DriverItem=(Type=Bone,Name="head")
    FalloffRegionScaleFactors=(PositiveWidth=1.5,NegativeWidth=1.5,PositiveHeight=3,NegativeHeight=3)
    FalloffSize=0.4
    OptionalParentItem=(Type=Bone,Name="neck_01")
    RotationOffset=(X=-90,Y=0,Z=0)
    ExecutePin=@RigUnit_BeginExecution.ExecutePin
   )
   > Set Transforms for Bone_1_1_1_1_1_1 = RigVMFunctionReferenceNode(
    Joint=(Type=Bone,Name="dyn_hat_MSTR")
    Rotate X=0
    Rotate Y=-20
    Rotate Z=0
    Scale X=0
    Scale Y=0
    Scale Z=0.1
    Translate X=0
    Translate Y=0
    Translate Z=0
    Weight=@n1.OutputParam
   )
   > n2 = &RigUnit_SphericalPoseReader(
    ActiveRegionScaleFactors=(PositiveWidth=1,NegativeWidth=1,PositiveHeight=1,NegativeHeight=1)
    ActiveRegionSize=0.1
    Debug=(bDrawDebug=True,DebugScale=25,DebugSegments=20,DebugThickness=0.25)
    DriverAxis=(X=1,Y=0,Z=0)
    DriverItem=(Type=Bone,Name="head")
    FalloffRegionScaleFactors=(PositiveWidth=1,NegativeWidth=1,PositiveHeight=1,NegativeHeight=1)
    FalloffSize=0.5
    OptionalParentItem=(Type=Bone,Name="None")
    RotationOffset=(X=-30,Y=0,Z=-10)
   )
   > Set Transforms for Bone_1_1_1_1_1_1_1_1 = RigVMFunctionReferenceNode(
    Joint=(Type=Bone,Name="dyn_back_brim")
    Rotate X=0
    Rotate Y=20
    Rotate Z=40
    Scale X=0
    Scale Y=0
    Scale Z=0
    Translate X=-10
    Translate Y=0
    Translate Z=0
    Weight=@n2.OutputParam
   )
   > &RigUnit_RotationConstraintLocalSpaceOffset(
    AdvancedSettings=(RotationOrderForFilter=XZY)
    Child=(Type=Bone,Name="dyn_bk_hair_01")
    Filter=(bX=True,bY=True,bZ=True)
    Parents=((Item=(Type=Bone,Name="spine_05"),Weight=1))
    Weight=@n2.OutputParam
   )
   > &RigUnit_RotationConstraintLocalSpaceOffset(
    AdvancedSettings=(RotationOrderForFilter=XZY)
    Child=(Type=Bone,Name="dyn_r_bk_hair_01")
    Filter=(bX=True,bY=True,bZ=True)
    Parents=((Item=(Type=Bone,Name="spine_05"),Weight=1))
    Weight=@n2.OutputParam
   )
   > &RigUnit_RotationConstraintLocalSpaceOffset(
    AdvancedSettings=(RotationOrderForFilter=XZY)
    Child=(Type=Bone,Name="dyn_l_bk_hair_01")
    Filter=(bX=True,bY=True,bZ=True)
    Parents=((Item=(Type=Bone,Name="spine_05"),Weight=1))
    Weight=@n2.OutputParam
   )
   > &RigUnit_RotationConstraintLocalSpaceOffset(
    AdvancedSettings=(RotationOrderForFilter=XZY)
    Child=(Type=Bone,Name="dyn_bk_hair_02")
    Filter=(bX=True,bY=True,bZ=True)
    Parents=((Item=(Type=Bone,Name="spine_05"),Weight=1))
    Weight=@n2.OutputParam
   )
   > &RigUnit_RotationConstraintLocalSpaceOffset(
    AdvancedSettings=(RotationOrderForFilter=XZY)
    Child=(Type=Bone,Name="dyn_r_bk_hair_02")
    Filter=(bX=True,bY=True,bZ=True)
    Parents=((Item=(Type=Bone,Name="spine_05"),Weight=1))
    Weight=@n2.OutputParam
   )
   > &RigUnit_RotationConstraintLocalSpaceOffset(
    AdvancedSettings=(RotationOrderForFilter=XZY)
    Child=(Type=Bone,Name="dyn_l_bk_hair_02")
    Filter=(bX=True,bY=True,bZ=True)
    Parents=((Item=(Type=Bone,Name="spine_05"),Weight=1))
    Weight=@n2.OutputParam
   )
  ]}
]
hierarchy [
 root {loc=<0 0 0>, rot=<0 0 0 1>}
 pelvis {parent=root, loc=<0 -0.5301 98.0571>, rot=<0.0127 0.707 -0.0127 -0.707>}
 spine_01 {parent=pelvis, loc=<6.1208 -0.9328 0>, rot=<0 0 0.034 -0.9994>}
 spine_02 {parent=spine_01, loc=<6.5996 -0.2095 0>, rot=<0 0 0.0238 -0.9997>}
 spine_03 {parent=spine_02, loc=<6.4914 -0.1237 0>, rot=<0 0 0.0149 -0.9999>}
 spine_04 {parent=spine_03, loc=<6.9598 0.0794 0>, rot=<0 0 -0.0132 -0.9999>}
 spine_05 {parent=spine_04, loc=<6.9987 0 0>, rot=<0 0 -0.0015 -1>}
 clavicle_l {parent=spine_05, loc=<5.4831 -0.1053 -3.2946>, rot=<0.0938 -0.7388 0.0307 -0.6666>}
 upperarm_l {parent=clavicle_l, loc=<9.9486 0 0>, rot=<0.0371 -0.2992 -0.0705 -0.9509>}
 upperarm_twist_01_l {parent=upperarm_l, loc=<0.0041 0 0>, rot=<0 0 0 1>}
 deform_upperarm_bk_base_l {parent=upperarm_twist_01_l, loc=<0.9473 -4.46 0.3113>, rot=<0.0004 -0.0034 -0.0158 -0.9999>}
 deform_upperarm_bk_l {parent=deform_upperarm_bk_base_l, loc=<0 0 0>, rot=<0 0 0 1>}
 deform_upperarm_up_base_l {parent=upperarm_twist_01_l, loc=<-1.6653 0.3195 3.9629>, rot=<0.0004 -0.0034 -0.0158 -0.9999>}
 deform_upperarm_up_l {parent=deform_upperarm_up_base_l, loc=<0 0 0>, rot=<0 0 0 1>}
 deform_upperarm_fr_base_l {parent=upperarm_twist_01_l, loc=<1.4228 5.0263 0.3212>, rot=<0.0004 -0.0034 -0.0158 -0.9999>}
 deform_upperarm_fr_l {parent=deform_upperarm_fr_base_l, loc=<0 0 0>, rot=<0 0 0 1>}
 lowerarm_l {parent=upperarm_l, loc=<26.4793 0 0>, rot=<0 0 -0.2382 -0.9712>}
 hand_l {parent=lowerarm_l, loc=<22.7458 0 0>, rot=<-0.6747 -0.0182 0.0605 -0.7354>}
 thumb_01_l {parent=hand_l, loc=<1.3118 -0.3191 -2.459>, rot=<0.6546 -0.4042 0.0664 -0.6354>}
 thumb_02_l {parent=thumb_01_l, loc=<5.9322 -0.0253 0.1416>, rot=<0.0121 0.0046 0.2964 -0.955>}
 thumb_03_l {parent=thumb_02_l, loc=<2.2223 0.3682 0>, rot=<0 0 0.0172 -0.9999>}
 index_metacarpal_l {parent=hand_l, loc=<1.6535 0.0183 -2.96>, rot=<0.0158 -0.0124 -0.0173 -0.9996>}
 index_01_l {parent=index_metacarpal_l, loc=<8.3371 0.1718 -0.2461>, rot=<0.0363 0.0499 0.142 -0.9879>}
 index_02_l {parent=index_01_l, loc=<3.4919 -0.3004 0>, rot=<0 0 0.2948 -0.9556>}
 index_03_l {parent=index_02_l, loc=<2.5578 -0.123 0>, rot=<0 0 0.1968 -0.9804>}
 middle_metacarpal_l {parent=hand_l, loc=<2.1856 -0.1725 -0.5984>, rot=<0.006 0.0168 -0.0329 -0.9993>}
 middle_01_l {parent=middle_metacarpal_l, loc=<7.6374 0.346 -0.1508>, rot=<-0.0085 0.0328 0.1813 -0.9829>}
 middle_02_l {parent=middle_01_l, loc=<3.9108 0 0>, rot=<0 0 0.2808 -0.9598>}
 middle_03_l {parent=middle_02_l, loc=<2.6103 0.0517 0>, rot=<0 0 0.1446 -0.9895>}
 ring_metacarpal_l {parent=hand_l, loc=<2.2074 -0.3625 1.3638>, rot=<-0.0284 0.0722 -0.0458 -0.9959>}
 ring_01_l {parent=ring_metacarpal_l, loc=<7.2483 0.0329 -0.1426>, rot=<-0.0456 -0.0277 0.2951 -0.954>}
 ring_02_l {parent=ring_01_l, loc=<3.7208 0.4647 0>, rot=<0 0 0.1583 -0.9874>}
 ring_03_l {parent=ring_02_l, loc=<2.4152 0.114 0>, rot=<0 0 0.1504 -0.9886>}
 pinky_metacarpal_l {parent=hand_l, loc=<1.7616 -0.9753 3.3814>, rot=<-0.0397 0.0814 -0.069 -0.9935>}
 pinky_01_l {parent=pinky_metacarpal_l, loc=<7.1933 -0.4205 -0.2601>, rot=<-0.0487 -0.0063 0.2556 -0.9655>}
 pinky_02_l {parent=pinky_01_l, loc=<2.8274 0.0893 0>, rot=<0 0 0.2474 -0.9689>}
 pinky_03_l {parent=pinky_02_l, loc=<2.2046 0.0307 0>, rot=<0 0 0.125 -0.9922>}
 weapon_l {parent=hand_l, loc=<-0.1974 -3.4111 0.3811>, rot=<0 0 0 1>}
 deform_wrist_dn_base_l {parent=hand_l, loc=<0 -1.5326 0>, rot=<0 0 0 -1>}
 deform_wrist_dn_l {parent=deform_wrist_dn_base_l, loc=<0 0 0>, rot=<0 0 0 1>}
 deform_wrist_up_base_l {parent=hand_l, loc=<0 2.3009 0>, rot=<0 0 0 -1>}
 deform_wrist_up_l {parent=deform_wrist_up_base_l, loc=<0 0 0>, rot=<0 0 0 1>}
 lowerarm_twist_02_l {parent=lowerarm_l, loc=<9.6199 0 0>, rot=<0 0 0 1>}
 lowerarm_twist_01_l {parent=lowerarm_l, loc=<17.8549 0 0>, rot=<0 0 0 1>}
 deform_elbow_in_base_l {parent=lowerarm_l, loc=<0.1385 3.6052 -0.2182>, rot=<0.0059 0.0117 -0.0071 -0.9999>}
 deform_elbow_in_l {parent=deform_elbow_in_base_l, loc=<0 0 0>, rot=<0 0 0 1>}
 deform_elbow_out_base_l {parent=lowerarm_l, loc=<0.0565 -3.9993 0.1259>, rot=<0.0059 0.0117 -0.0071 -0.9999>}
 deform_elbow_out_l {parent=deform_elbow_out_base_l, loc=<0 0 0>, rot=<0 0 0 1>}
 upperarm_twist_02_l {parent=upperarm_l, loc=<13.7941 0 0>, rot=<0 0 0 1>}
 neck_01 {parent=spine_05, loc=<11.3138 0.1529 0>, rot=<0 0 -0.2003 -0.9797>}
 neck_02 {parent=neck_01, loc=<4.5627 0 0>, rot=<0 0 0.0504 -0.9987>}
 head {parent=neck_02, loc=<5.3914 0 0>, rot=<0 0 0.115 -0.9934>}
 dyn_hat_MSTR {parent=head, loc=<8.6349 -1.2138 0.175>, rot=<-0.1237 -0.6961 -0.6952 -0.1297>}
 dyn_hat {parent=dyn_hat_MSTR, loc=<0 0 0>, rot=<0 0 0 1>}
 dyn_backBrim {parent=dyn_hat_MSTR, loc=<1.0668 0.0744 -9.4746>, rot=<0.722 0.025 -0.6914 -0.0002>}
 dyn_R_Brim {parent=dyn_hat_MSTR, loc=<0.4389 7.5969 -2.2832>, rot=<-0.5625 -0.4125 0.5713 0.4326>}
 dyn_back_brim {parent=dyn_hat_MSTR, loc=<0 0 0>, rot=<0 0 0 1>}
 dyn_hairBang_02 {parent=head, loc=<9.6853 8.6931 0.8517>, rot=<0.1055 0.6475 0.7451 0.1202>}
 dyn_hairBang_03 {parent=head, loc=<9.7442 8.4886 -0.9508>, rot=<-0.0786 -0.7319 -0.6642 -0.1302>}
 dyn_hairBang_01 {parent=head, loc=<9.9277 7.6328 3.1197>, rot=<0.1544 0.4928 0.855 0.0472>}
 dyn_hairBang_04 {parent=head, loc=<9.7596 7.6323 -3.2623>, rot=<-0.065 -0.862 -0.485 -0.1322>}
 dyn_l_bk_hairSide_MSTR {parent=head, loc=<2.2833 -3.7376 -6.4719>, rot=<-0.0282 -0.9887 0.1433 -0.0326>}
 dyn_l_bk_hairSide_01 {parent=dyn_l_bk_hairSide_MSTR, loc=<0 0 0>, rot=<0 0 0 1>}
 dyn_l_bk_hairSide_02 {parent=dyn_l_bk_hairSide_01, loc=<7.5024 0 0.1595>, rot=<0 0.0174 0 -0.9998>}
 dyn_l_bk_hair_MSTR {parent=head, loc=<2.0668 -6.1211 -5.4159>, rot=<0.12 -0.7791 0.5966 -0.1505>}
 dyn_l_bk_hair_01 {parent=dyn_l_bk_hair_MSTR, loc=<0 0 0>, rot=<0 0 0 1>}
 dyn_l_bk_hair_02 {parent=dyn_l_bk_hair_01, loc=<7.5024 0 0.1595>, rot=<0 -0.0734 0 -0.9973>}
 dyn_l_bk_hair_03 {parent=dyn_l_bk_hair_02, loc=<8.472 0 0.1061>, rot=<0 -0.0121 0 -0.9999>}
 dyn_bk_hair_MSTR {parent=head, loc=<1.8478 -8.6205 0.0553>, rot=<0.1052 -0.6849 0.713 -0.1074>}
 dyn_bk_hair_01 {parent=dyn_bk_hair_MSTR, loc=<0 0 0>, rot=<0 0 0 1>}
 dyn_bk_hair_02 {parent=dyn_bk_hair_01, loc=<8.9616 0 -0.0565>, rot=<0 -0.0734 0 -0.9973>}
 dyn_bk_hair_03 {parent=dyn_bk_hair_02, loc=<8.472 0 0.1061>, rot=<0 -0.0121 0 -0.9999>}
 dyn_r_bk_hair_MSTR {parent=head, loc=<3.0259 -6.3401 6.4412>, rot=<0.1258 -0.6123 0.778 -0.0624>}
 dyn_r_bk_hair_01 {parent=dyn_r_bk_hair_MSTR, loc=<0 0 0>, rot=<0 0 0 1>}
 dyn_r_bk_hair_02 {parent=dyn_r_bk_hair_01, loc=<7.8946 0 -0.0306>, rot=<0 -0.0121 0 -0.9999>}
 dyn_r_bk_hair_03 {parent=dyn_r_bk_hair_02, loc=<7.63 0 0.1266>, rot=<0 -0.0121 0 -0.9999>}
 dyn_r_bk_hairSide_MSTR {parent=head, loc=<2.6388 -2.4946 6.5764>, rot=<0.0171 0.0167 0.9996 -0.0131>}
 dyn_r_bk_hairSide_01 {parent=dyn_r_bk_hairSide_MSTR, loc=<0 0 0>, rot=<0 0 0 1>}
 dyn_r_bk_hairSide_02 {parent=dyn_r_bk_hairSide_01, loc=<7.5024 0 0.1595>, rot=<0 0.0174 0 -0.9998>}
 dyn_r_sideHair_03 {parent=head, loc=<5.868 5.6317 6.0842>, rot=<0.0113 -0.9943 0.0737 0.0762>}
 dyn_r_sideHair_04 {parent=head, loc=<4.2965 1.7773 6.6099>, rot=<-0.0158 0.0669 0.9976 0.0121>}
 dyn_l_sideHair_02 {parent=head, loc=<4.9401 1.8486 -6.4919>, rot=<-0.0225 -0.9961 -0.0848 -0.0061>}
 dyn_noseRing_MSTR {parent=head, loc=<0.0775 8.7332 -0.0349>, rot=<-0.199 -0.6909 -0.6742 -0.1689>}
 dyn_noseRing {parent=dyn_noseRing_MSTR, loc=<0 0 0>, rot=<0 0 0 1>}
 dyn_l_ear_MSTR {parent=head, loc=<0.4958 0.0928 -6.7097>, rot=<0.1546 -0.9577 -0.236 0.0568>}
 dyn_l_ear {parent=dyn_l_ear_MSTR, loc=<0 0 0>, rot=<0 0 0 1>}
 dyn_r_ear_MSTR {parent=head, loc=<0.5289 0.1682 6.6709>, rot=<-0.0155 0.285 0.9353 -0.2092>}
 dyn_r_ear {parent=dyn_r_ear_MSTR, loc=<0 0 0>, rot=<0 0 0 1>}
 dyn_r_sideHair_01 {parent=head, loc=<5.868 5.6317 6.0842>, rot=<0.0113 -0.9943 0.0737 0.0762>}
 dyn_r_sideHair_02 {parent=head, loc=<4.2965 1.7773 6.6099>, rot=<-0.0158 0.0669 0.9976 0.0121>}
 dyn_l_sideHair_01 {parent=head, loc=<4.9401 1.8486 -6.4919>, rot=<-0.0225 -0.9961 -0.0848 -0.0061>}
 clavicle_r {parent=spine_05, loc=<5.4831 -0.1053 3.2946>, rot=<0.7388 0.0938 -0.6666 -0.0307>}
 upperarm_r {parent=clavicle_r, loc=<-9.9486 0 0>, rot=<0.0371 -0.2992 -0.0705 -0.9509>}
 upperarm_twist_01_r {parent=upperarm_r, loc=<-0.0041 0 0>, rot=<0 0 0 -1>}
 deform_upperarm_bk_base_r {parent=upperarm_twist_01_r, loc=<-0.9473 4.46 -0.3113>, rot=<0.0004 -0.0034 -0.0158 -0.9999>}
 deform_upperarm_bk_r {parent=deform_upperarm_bk_base_r, loc=<0 0 0>, rot=<0 0 0 1>}
 deform_upperarm_up_base_r {parent=upperarm_twist_01_r, loc=<1.6653 -0.3195 -3.9628>, rot=<0.0004 -0.0034 -0.0158 -0.9999>}
 deform_upperarm_up_r {parent=deform_upperarm_up_base_r, loc=<0 0 0>, rot=<0 0 0 1>}
 deform_upperarm_fr_base_r {parent=upperarm_twist_01_r, loc=<-1.4228 -5.0263 -0.3212>, rot=<0.0004 -0.0034 -0.0158 -0.9999>}
 deform_upperarm_fr_r {parent=deform_upperarm_fr_base_r, loc=<0 0 0>, rot=<0 0 0 1>}
 lowerarm_r {parent=upperarm_r, loc=<-26.4793 0 0>, rot=<0 0 -0.2382 -0.9712>}
 hand_r {parent=lowerarm_r, loc=<-22.7458 0 0>, rot=<-0.6747 -0.0182 0.0605 -0.7354>}
 thumb_01_r {parent=hand_r, loc=<-1.3118 0.3191 2.459>, rot=<0.6546 -0.4042 0.0664 -0.6354>}
 thumb_02_r {parent=thumb_01_r, loc=<-5.9322 0.0253 -0.1416>, rot=<0.0121 0.0046 0.2964 -0.955>}
 thumb_03_r {parent=thumb_02_r, loc=<-2.2223 -0.3682 0>, rot=<0 0 0.0172 -0.9999>}
 index_metacarpal_r {parent=hand_r, loc=<-1.6535 -0.0183 2.96>, rot=<0.0158 -0.0124 -0.0173 -0.9996>}
 index_01_r {parent=index_metacarpal_r, loc=<-8.3371 -0.1718 0.2461>, rot=<0.0363 0.0499 0.142 -0.9879>}
 index_02_r {parent=index_01_r, loc=<-3.4919 0.3004 0>, rot=<0 0 0.2948 -0.9556>}
 index_03_r {parent=index_02_r, loc=<-2.5578 0.123 0>, rot=<0 0 0.1968 -0.9804>}
 middle_metacarpal_r {parent=hand_r, loc=<-2.1856 0.1725 0.5984>, rot=<0.006 0.0168 -0.0329 -0.9993>}
 middle_01_r {parent=middle_metacarpal_r, loc=<-7.6374 -0.346 0.1508>, rot=<-0.0085 0.0328 0.1813 -0.9829>}
 middle_02_r {parent=middle_01_r, loc=<-3.9108 0 0>, rot=<0 0 0.2808 -0.9598>}
 middle_03_r {parent=middle_02_r, loc=<-2.6103 -0.0517 0>, rot=<0 0 0.1446 -0.9895>}
 ring_metacarpal_r {parent=hand_r, loc=<-2.2074 0.3625 -1.3638>, rot=<-0.0284 0.0722 -0.0458 -0.9959>}
 ring_01_r {parent=ring_metacarpal_r, loc=<-7.2483 -0.0329 0.1426>, rot=<-0.0456 -0.0277 0.2951 -0.954>}
 ring_02_r {parent=ring_01_r, loc=<-3.7208 -0.4647 0>, rot=<0 0 0.1583 -0.9874>}
 ring_03_r {parent=ring_02_r, loc=<-2.4152 -0.114 0>, rot=<0 0 0.1504 -0.9886>}
 pinky_metacarpal_r {parent=hand_r, loc=<-1.7616 0.9753 -3.3814>, rot=<-0.0397 0.0814 -0.069 -0.9935>}
 pinky_01_r {parent=pinky_metacarpal_r, loc=<-7.1933 0.4205 0.2601>, rot=<-0.0487 -0.0063 0.2556 -0.9655>}
 pinky_02_r {parent=pinky_01_r, loc=<-2.8274 -0.0893 0>, rot=<0 0 0.2474 -0.9689>}
 pinky_03_r {parent=pinky_02_r, loc=<-2.2046 -0.0307 0>, rot=<0 0 0.125 -0.9922>}
 weapon_r {parent=hand_r, loc=<0.1974 3.4111 -0.3811>, rot=<0 0 0 1>}
 deform_wrist_dn_base_r {parent=hand_r, loc=<-0.0001 1.5328 0>, rot=<0 0 0 -1>}
 deform_wrist_dn_r {parent=deform_wrist_dn_base_r, loc=<0 0 0>, rot=<0 0 0 1>}
 deform_wrist_up_base_r {parent=hand_r, loc=<-0.0002 -2.3007 -0.0001>, rot=<0 0 0 -1>}
 deform_wrist_up_r {parent=deform_wrist_up_base_r, loc=<0 0 0>, rot=<0 0 0 1>}
 lowerarm_twist_02_r {parent=lowerarm_r, loc=<-9.6199 0 0>, rot=<0 0 0 -1>}
 lowerarm_twist_01_r {parent=lowerarm_r, loc=<-17.8549 0 0>, rot=<0 0 0 -1>}
 deform_elbow_in_base_r {parent=lowerarm_r, loc=<-0.1385 -3.6052 0.2182>, rot=<0.0059 0.0117 -0.0071 -0.9999>}
 deform_elbow_in_r {parent=deform_elbow_in_base_r, loc=<0 0 0>, rot=<0 0 0 1>}
 deform_elbow_out_base_r {parent=lowerarm_r, loc=<-0.0565 3.9993 -0.126>, rot=<0.0059 0.0117 -0.0071 -0.9999>}
 deform_elbow_out_r {parent=deform_elbow_out_base_r, loc=<0 0 0>, rot=<0 0 0 1>}
 upperarm_twist_02_r {parent=upperarm_r, loc=<-13.7941 0 0>, rot=<0 0 0 -1>}
 dyn_necklace_04 {parent=spine_05, loc=<9.7253 1.7428 -0.0346>, rot=<0.0046 0.0053 0.9136 0.4065>}
 dyn_necklace_02 {parent=dyn_necklace_04, loc=<7.4757 0 0>, rot=<-0.0034 -0.0016 -0.0988 -0.9951>}
 dyn_necklace_03 {parent=dyn_necklace_02, loc=<4.5386 0.5153 0.0152>, rot=<-0.0262 -0.006 -0.1014 -0.9945>}
 deform_pec_aim_base_l {parent=spine_04, loc=<3.1942 9.3086 -1.2237>, rot=<0.0614 -0.4031 0.3295 -0.8516>}
 deform_pec_aim_l {parent=deform_pec_aim_base_l, loc=<0 0 0>, rot=<0 0 0 1>}
 deform_pec_base_l {parent=deform_pec_aim_l, loc=<9.3665 0 0>, rot=<0 0 0 1>}
 deform_pec_l {parent=deform_pec_base_l, loc=<0 0 0>, rot=<0 0 0 1>}
 deform_pec_aim_base_r {parent=spine_04, loc=<3.1938 9.3085 1.2237>, rot=<-0.4031 -0.0614 0.8516 0.3295>}
 deform_pec_aim_r {parent=deform_pec_aim_base_r, loc=<0 0 0>, rot=<0 0 0 1>}
 deform_pec_base_r {parent=deform_pec_aim_r, loc=<-9.3667 -0.0003 -0.0004>, rot=<0 0 0 1>}
 deform_pec_r {parent=deform_pec_base_r, loc=<0 0 0>, rot=<0 0 0 1>}
 dyn_l_bk_dress_MSTR {parent=pelvis, loc=<33.4636 -13.1591 -5.8386>, rot=<0.0085 -0.7117 0.7021 -0.0226>}
 dyn_l_bk_dress_01 {parent=dyn_l_bk_dress_MSTR, loc=<0 0 0>, rot=<0 0 0 1>}
 dyn_l_bk_dress_02 {parent=dyn_l_bk_dress_01, loc=<12.7819 0 -0.1492>, rot=<0 -0.0121 0 -0.9999>}
 dyn_l_bk_dress_03 {parent=dyn_l_bk_dress_02, loc=<10.3593 0 0.0603>, rot=<0 -0.0121 0 -0.9999>}
 dyn_l_bk_dress_04 {parent=dyn_l_bk_dress_03, loc=<17.6639 0 -0.6348>, rot=<0 -0.0169 0 -0.9999>}
 dyn_l_bk_dress_05 {parent=dyn_l_bk_dress_04, loc=<19.111 0 -0.6964>, rot=<0 -0.0121 0 -0.9999>}
 dyn_r_bk_dress_MSTR {parent=pelvis, loc=<33.2171 -13.3361 6.7541>, rot=<0.0085 -0.7117 0.7021 -0.0226>}
 dyn_r_bk_dress_01 {parent=dyn_r_bk_dress_MSTR, loc=<0 0 0>, rot=<0 0 0 1>}
 dyn_r_bk_dress_02 {parent=dyn_r_bk_dress_01, loc=<12.7819 0 -0.1492>, rot=<0 -0.0121 0 -0.9999>}
 dyn_r_bk_dress_03 {parent=dyn_r_bk_dress_02, loc=<10.3593 0 0.0603>, rot=<0 -0.0121 0 -0.9999>}
 dyn_r_bk_dress_04 {parent=dyn_r_bk_dress_03, loc=<17.6639 0 -0.6348>, rot=<0 -0.0169 0 -0.9999>}
 dyn_r_bk_dress_05 {parent=dyn_r_bk_dress_04, loc=<19.111 0 -0.6964>, rot=<0 -0.0121 0 -0.9999>}
 thigh_l {parent=pelvis, loc=<-1.0855 -0.2109 -8.7053>, rot=<0.0009 0.0345 0.0155 -0.9993>}
 calf_l {parent=thigh_l, loc=<-42.5268 0 0>, rot=<0 0 -0.0657 -0.9978>}
 foot_l {parent=calf_l, loc=<-46.8037 0 0>, rot=<0.0422 -0.04 0.0613 -0.9964>}
 ball_l {parent=foot_l, loc=<-3.5658 10.8066 -0.1544>, rot=<0.0001 -0.0001 -0.7063 -0.7079>}
 calf_twist_02_l {parent=calf_l, loc=<-18.8064 -0.0967 0.0397>, rot=<0.0237 0.0095 -0.0035 -0.9997>}
 calf_twist_01_l {parent=calf_l, loc=<-35.813 -0.0967 0.0397>, rot=<0.0237 0.0095 -0.0035 -0.9997>}
 dyn_l_lace_MSTR {parent=calf_twist_01_l, loc=<0.2594 4.3621 -0.6163>, rot=<0.1524 0.6879 0.7025 0.1001>}
 dyn_l_loop_03 {parent=dyn_l_lace_MSTR, loc=<0.2951 1.0365 -0.1162>, rot=<-0.2041 -0.3417 -0.3424 -0.8511>}
 dyn_l_loop_04 {parent=dyn_l_loop_03, loc=<2.6006 0 0>, rot=<0 0 0 1>}
 dyn_l_loop_01 {parent=dyn_l_lace_MSTR, loc=<-0.0044 -0.957 -0.0207>, rot=<0.0677 -0.1569 0.528 -0.8319>}
 dyn_l_loop_02 {parent=dyn_l_loop_01, loc=<2.6006 0 0>, rot=<0 0 0 1>}
 dyn_l_lace_03 {parent=dyn_l_lace_MSTR, loc=<0.7542 0.9672 -0.159>, rot=<-0.271 -0.154 -0.1979 -0.9293>}
 dyn_l_lace_04 {parent=dyn_l_lace_03, loc=<2.4815 0 0>, rot=<0 0.0712 0 -0.9975>}
 dyn_l_lace_01 {parent=dyn_l_lace_MSTR, loc=<0.4774 -0.5248 0.1423>, rot=<0.0758 -0.0986 0.2902 -0.9488>}
 dyn_l_lace_02 {parent=dyn_l_lace_01, loc=<2.4815 0 0>, rot=<-0.0043 -0.0785 -0.0539 -0.9954>}
 deform_knee_in_base_l {parent=calf_l, loc=<0.5405 -5.0592 -0.5944>, rot=<-0.0003 -0.0151 0.0267 -0.9995>}
 deform_knee_in_l {parent=deform_knee_in_base_l, loc=<0 0 0>, rot=<0 0 0 1>}
 deform_knee_out_base_l {parent=calf_l, loc=<0.2138 3.9943 -0.8167>, rot=<-0.0003 -0.0151 0.0267 -0.9995>}
 deform_knee_out_l {parent=deform_knee_out_base_l, loc=<0 0 0>, rot=<0 0 0 1>}
 thigh_twist_01_l {parent=thigh_l, loc=<-21.629 -0.0919 -0.0002>, rot=<0.0013 0 0 -1>}
 deform_glute_base_l {parent=thigh_l, loc=<-6.0004 -8.9997 -0.0128>, rot=<0.0004 0.0009 -0.4617 -0.887>}
 deform_glute_l {parent=deform_glute_base_l, loc=<-0.0006 -0.0014 0>, rot=<0 0 0 1>}
 deform_groin_base_l {parent=thigh_l, loc=<0.4808 5.6776 -0.2472>, rot=<0 0.001 0 -1>}
 deform_groin_l {parent=deform_groin_base_l, loc=<-3 12 0>, rot=<0 0 0 1>, scale=<0.5 1 1>}
 thigh_r {parent=pelvis, loc=<-1.0855 -0.2109 8.7053>, rot=<0.0345 -0.0009 0.9993 0.0155>}
 calf_r {parent=thigh_r, loc=<42.5268 0 0>, rot=<0 0 -0.0657 -0.9978>}
 foot_r {parent=calf_r, loc=<46.8037 0 0>, rot=<0.0422 -0.04 0.0613 -0.9964>}
 ball_r {parent=foot_r, loc=<3.5658 -10.8066 0.1544>, rot=<0.0001 -0.0001 -0.7063 -0.7079>}
 calf_twist_02_r {parent=calf_r, loc=<18.8064 0.0967 -0.0397>, rot=<0.0237 0.0095 -0.0035 -0.9997>}
 calf_twist_01_r {parent=calf_r, loc=<35.813 0.0967 -0.0397>, rot=<0.0237 0.0095 -0.0035 -0.9997>}
 dyn_r_lace_MSTR {parent=calf_twist_01_r, loc=<-0.5517 -4.4282 0.5248>, rot=<-0.6666 0.094 0.1563 -0.7227>}
 dyn_r_lace_03 {parent=dyn_r_lace_MSTR, loc=<-0.1373 0.7199 0.2464>, rot=<-0.1669 0.0045 -0.5667 -0.8068>}
 dyn_r_lace_04 {parent=dyn_r_lace_03, loc=<2.6006 0 0>, rot=<0 0 0.1166 -0.9932>}
 dyn_r_loop_01 {parent=dyn_r_lace_MSTR, loc=<0.3301 -0.7687 -0.0609>, rot=<0.0024 -0.1073 0.3341 -0.9364>}
 dyn_r_loop_02 {parent=dyn_r_loop_01, loc=<2.6006 0 0>, rot=<0 0 0 1>}
 dyn_r_loop_03 {parent=dyn_r_lace_MSTR, loc=<0.4771 0.6526 0.0131>, rot=<-0.0966 -0.089 -0.4507 -0.883>}
 dyn_r_loop_04 {parent=dyn_r_loop_03, loc=<2.4815 0 0>, rot=<0 0.0712 0 -0.9975>}
 dyn_r_lace_01 {parent=dyn_r_lace_MSTR, loc=<0.0625 -0.4621 0.7101>, rot=<0.2656 0.0375 0.4412 -0.8564>}
 dyn_r_lace_02 {parent=dyn_r_lace_01, loc=<2.4815 0 0>, rot=<0.0344 0.0595 -0.1608 -0.9846>}
 deform_knee_in_base_r {parent=calf_r, loc=<-0.5405 5.0592 0.5943>, rot=<-0.0003 -0.0151 0.0267 -0.9995>}
 deform_knee_in_r {parent=deform_knee_in_base_r, loc=<0 0 0>, rot=<0 0 0 1>}
 deform_knee_out_base_r {parent=calf_r, loc=<-0.2137 -3.9943 0.8168>, rot=<-0.0003 -0.0151 0.0267 -0.9995>}
 deform_knee_out_r {parent=deform_knee_out_base_r, loc=<0 0 0>, rot=<0 0 0 1>}
 thigh_twist_01_r {parent=thigh_r, loc=<21.629 0.0919 0.0002>, rot=<0.0013 0 0 -1>}
 deform_glute_base_r {parent=thigh_r, loc=<6.0002 8.9997 0.0129>, rot=<0.0004 0.0009 -0.4617 -0.887>}
 deform_glute_r {parent=deform_glute_base_r, loc=<2.0001 4.9998 0>, rot=<0 0 0 1>}
 deform_groin_base_r {parent=thigh_r, loc=<-0.4808 -5.6776 0.2472>, rot=<0 0.001 0 -1>}
 deform_groin_r {parent=deform_groin_base_r, loc=<0 0 0>, rot=<0 0 0 1>}
 dyn_squid_MSTR {parent=thigh_r, loc=<-0.0181 -0.0006 8.7946>, rot=<0 0.1339 0 -0.991>}
 dyn_squid_base {parent=dyn_squid_MSTR, loc=<9.3133 2.3928 0.152>, rot=<-0.0045 -0.1679 0.0032 -0.9858>}
 dyn_squid_chain_01 {parent=dyn_squid_base, loc=<9.5685 0 0>, rot=<0 0 0 1>}
 dyn_squid_chain_02 {parent=dyn_squid_chain_01, loc=<10.5759 0 0>, rot=<0 0 0 1>}
 dyn_squid_tenticle_b_01 {parent=dyn_squid_base, loc=<4.133 3.467 2.9824>, rot=<0.2706 0.0163 0.058 -0.9608>}
 dyn_squid_tenticle_b_02 {parent=dyn_squid_tenticle_b_01, loc=<10.5759 0 0>, rot=<0 0 0 1>}
 dyn_squid_tenticle_a_01 {parent=dyn_squid_base, loc=<4.6526 -3.9182 3.4249>, rot=<-0.2759 -0.0626 0.0049 -0.9591>}
 dyn_squid_tenticle_a_02 {parent=dyn_squid_tenticle_a_01, loc=<10.5759 0 0>, rot=<0 0 0 1>}
 dyn_l_fr_dress_MSTR {parent=pelvis, loc=<33.3269 2.6801 -5.6187>, rot=<0.111 0.6913 0.7034 0.123>}
 dyn_l_fr_dress_01 {parent=dyn_l_fr_dress_MSTR, loc=<0 0 0>, rot=<0 0 0 1>}
 dyn_l_fr_dress_02 {parent=dyn_l_fr_dress_01, loc=<12.7819 0 -0.1492>, rot=<0 -0.0431 0 -0.9991>}
 dyn_l_fr_dress_03 {parent=dyn_l_fr_dress_02, loc=<10.3593 0 0.0603>, rot=<0 -0.0431 0 -0.9991>}
 dyn_l_fr_dress_04 {parent=dyn_l_fr_dress_03, loc=<17.6639 0 -0.6348>, rot=<0 -0.0479 0 -0.9989>}
 dyn_l_fr_dress_05 {parent=dyn_l_fr_dress_04, loc=<19.111 0 -0.6964>, rot=<0 -0.0431 0 -0.9991>}
 dyn_r_fr_dress_MSTR {parent=pelvis, loc=<33.5602 2.4456 5.1242>, rot=<0.125 0.691 0.7036 0.1087>}
 dyn_r_fr_dress_01 {parent=dyn_r_fr_dress_MSTR, loc=<0 0 0>, rot=<0 0 0 1>}
 dyn_r_fr_dress_02 {parent=dyn_r_fr_dress_01, loc=<12.7819 0 -0.1492>, rot=<0 -0.0431 0 -0.9991>}
 dyn_r_fr_dress_03 {parent=dyn_r_fr_dress_02, loc=<10.3593 0 0.0603>, rot=<0 -0.0431 0 -0.9991>}
 dyn_r_fr_dress_04 {parent=dyn_r_fr_dress_03, loc=<17.6639 0 -0.6348>, rot=<0 -0.0479 0 -0.9989>}
 dyn_r_fr_dress_05 {parent=dyn_r_fr_dress_04, loc=<19.111 0 -0.6964>, rot=<0 -0.0431 0 -0.9991>}
 attach {parent=root, loc=<0 0 0>, rot=<0 0 0 1>}
 ik_foot_root {parent=root, loc=<0 0 0>, rot=<0 0 0 1>}
 ik_foot_l {parent=ik_foot_root, loc=<14.8262 -7.3545 8.296>, rot=<0.0357 0.7045 0.0284 -0.7082>}
 ik_foot_r {parent=ik_foot_root, loc=<-14.8262 -7.3545 8.296>, rot=<0.7082 0.0284 -0.7045 0.0357>}
 ik_hand_root {parent=root, loc=<0 0 0>, rot=<0 0 0 1>}
 ik_hand_gun {parent=ik_hand_root, loc=<-48.1743 3.3545 104.4381>, rot=<0.6276 0.1309 0.4183 -0.6434>}
 ik_hand_r {parent=ik_hand_gun, loc=<0 0 0>, rot=<0 0 0 1>}
 ik_hand_l {parent=ik_hand_gun, loc=<59.3271 67.6997 34.3524>, rot=<-0.6158 -0.7027 -0.3565 0>}
]
