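# Physics asset for the Gloom female medium character's hat and face accessory part, with capsule bodies for dynamic back-hair bones and the upper body.
# Left, center and right hair chains each run three angle-limited segments from a head-attached master, the right having larger capsules and tighter twist.
# Spine, neck, head, clavicle and upper-arm capsules, plus a right-shoulder deform bone, form the body shape, linked by angularly free constraints.
# An extra right side-hair master hangs freely from the head without child segments.
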PhysicsAsset /BRCosmetics/Characters/Player/Female/Medium/Bodies/F_MED_Gloom/Meshes/Parts/F_MED_Gloom_Hat_FaceAcc_Physics
bodies {
 clavicle_l [{shapeType=Capsule, length=6, radius=6, shapeName=clavicle_l_capsule, center=<5.0738 0 0>, rotation=<rot 90 0 0>}]
 clavicle_r [{shapeType=Capsule, length=6, radius=6, shapeName=clavicle_l_capsule, center=<-4.7067 0 0>, rotation=<rot 90 0 0>}]
 deform_upperarm_up_base_r [{shapeType=Capsule, length=6.3803, radius=3.6197, shapeName=deform_upperarm_up_base_r_capsule, rotation=<rot 0 -90 47.7995>}]
 dyn_bk_hair_01 [{shapeType=Capsule, length=4.2231, radius=1.7769, shapeName=dyn_bk_hair_01_capsule, center=<4.5156 0 0>, rotation=<rot 90 0 0>}]
 dyn_bk_hair_02 [{shapeType=Capsule, length=4.2231, radius=1.7769, shapeName=dyn_bk_hair_02_capsule, center=<4.5156 0 0>, rotation=<rot 90 0 0>}]
 dyn_bk_hair_03 [{shapeType=Capsule, length=4.2231, radius=1.7769, shapeName=dyn_bk_hair_03_capsule, center=<4.5156 0 0>, rotation=<rot 90 0 0>}]
 dyn_bk_hair_MSTR [{shapeType=Capsule, length=0.505, radius=0.505, shapeName=dyn_bk_hair_MSTR_capsule}]
 dyn_l_bk_hair_01 [{shapeType=Capsule, length=4.2231, radius=1.7769, shapeName=dyn_l_bk_hair_01_capsule, center=<4.5156 0 2>, rotation=<rot 90 0 0>}]
 dyn_l_bk_hair_02 [{shapeType=Capsule, length=3.2297, radius=2.7703, shapeName=dyn_l_bk_hair_02_capsule, center=<4.5156 0 2>, rotation=<rot 90 0 0>}]
 dyn_l_bk_hair_03 [{shapeType=Capsule, length=2.9934, radius=3.0066, shapeName=dyn_l_bk_hair_03_capsule, center=<4.5156 0 2>, rotation=<rot 90 0 0>}]
 dyn_l_bk_hair_MSTR [{shapeType=Capsule, length=0.505, radius=0.505, shapeName=dyn_l_bk_hair_MSTR_capsule}]
 dyn_r_bk_hairSide_MSTR [{shapeType=Capsule, length=0.505, radius=0.505, shapeName=dyn_r_bk_hairSide_MSTR_capsule, rotation=<rot 0 -90 0>}]
 dyn_r_bk_hair_01 [{shapeType=Capsule, length=2.0131, radius=7.0293, shapeName=dyn_r_bk_hair_01_capsule, center=<4.5156 0 3>, rotation=<rot 90 0 0>}]
 dyn_r_bk_hair_02 [{shapeType=Capsule, length=2.0131, radius=7.0293, shapeName=dyn_r_bk_hair_02_capsule, center=<4.5156 0 3>, rotation=<rot 90 0 0>}]
 dyn_r_bk_hair_03 [{shapeType=Capsule, length=2.0131, radius=7.0293, shapeName=dyn_r_bk_hair_03_capsule, center=<4.5156 0 3>, rotation=<rot 90 0 0>}]
 dyn_r_bk_hair_MSTR [{shapeType=Capsule, length=0.505, radius=0.505, shapeName=dyn_r_bk_hair_MSTR_capsule}]
 head [{shapeType=Capsule, length=7.9736, radius=5.8955, center=<1.8068 -1.0039 -0.1194>, rotation=<rot 90 0 -10.0004>}]
 neck_01 [{shapeType=Capsule, length=46.4501, radius=12.7458, shapeName=neck_01_capsule, center=<5.6177 5.3843 0>}]
 neck_02 [{shapeType=Capsule, length=0.321, radius=5.9274, shapeName=neck_01_capsule, center=<2.8325 -0.5621 0>, rotation=<rot 90 0 0>}]
 spine_04 [{shapeType=Capsule, length=23.6364, radius=8.5815, rotation=<rot 0 0 0.1926>}]
 spine_05 [{shapeType=Capsule, length=127.61, radius=8.8627}]
 upperarm_l [{shapeType=Capsule, length=13.2565, radius=9.4362, shapeName=upperarm_l_capsule, center=<13.2101 0 0>, rotation=<rot -90 0 0>}]
 upperarm_r [{shapeType=Capsule, length=13.4666, radius=9.2261, shapeName=upperarm_r_capsule, center=<-13.4087 0 0>, rotation=<rot -90 180 180>}]
}
constraints [
 {
  bone1Name=head
  bone2Name=neck_02
  swing1LimitDegrees=45
  swing1Motion=Free
  swing2LimitDegrees=45
  swing2Motion=Free
  twistLimitDegrees=45
  twistMotion=Free
 }
 {
  bone1Name=neck_02
  bone2Name=neck_01
  swing1LimitDegrees=45
  swing1Motion=Free
  swing2LimitDegrees=45
  swing2Motion=Free
  twistLimitDegrees=45
  twistMotion=Free
 }
 {
  bone1Name=neck_01
  bone2Name=spine_05
  swing1LimitDegrees=45
  swing1Motion=Free
  swing2LimitDegrees=45
  swing2Motion=Free
  twistLimitDegrees=45
  twistMotion=Free
 }
 {
  bone1Name=spine_05
  bone2Name=spine_04
  swing1LimitDegrees=45
  swing1Motion=Free
  swing2LimitDegrees=45
  swing2Motion=Free
  twistLimitDegrees=45
  twistMotion=Free
 }
 {bone1Name=dyn_l_bk_hair_MSTR, bone2Name=head, swing1LimitDegrees=10, swing2LimitDegrees=45, twistLimitDegrees=5}
 {bone1Name=dyn_bk_hair_MSTR, bone2Name=head, swing1LimitDegrees=10, swing2LimitDegrees=45, twistLimitDegrees=5}
 {bone1Name=dyn_r_bk_hair_MSTR, bone2Name=head, swing1LimitDegrees=10, swing2LimitDegrees=45, twistLimitDegrees=5}
 {bone1Name=dyn_l_bk_hair_01, bone2Name=dyn_l_bk_hair_MSTR, swing1LimitDegrees=30, swing2LimitDegrees=30, twistLimitDegrees=30}
 {bone1Name=dyn_l_bk_hair_02, bone2Name=dyn_l_bk_hair_01, swing1LimitDegrees=30, swing2LimitDegrees=30, twistLimitDegrees=30}
 {bone1Name=dyn_l_bk_hair_03, bone2Name=dyn_l_bk_hair_02, swing1LimitDegrees=30, swing2LimitDegrees=30, twistLimitDegrees=30}
 {bone1Name=dyn_bk_hair_01, bone2Name=dyn_bk_hair_MSTR, swing1LimitDegrees=30, swing2LimitDegrees=30, twistLimitDegrees=30}
 {bone1Name=dyn_bk_hair_02, bone2Name=dyn_bk_hair_01, swing1LimitDegrees=30, swing2LimitDegrees=30, twistLimitDegrees=30}
 {bone1Name=dyn_bk_hair_03, bone2Name=dyn_bk_hair_02, swing1LimitDegrees=30, swing2LimitDegrees=30, twistLimitDegrees=30}
 {bone1Name=dyn_r_bk_hair_01, bone2Name=dyn_r_bk_hair_MSTR, swing1LimitDegrees=30, swing2LimitDegrees=30, twistLimitDegrees=15}
 {bone1Name=dyn_r_bk_hair_02, bone2Name=dyn_r_bk_hair_01, swing1LimitDegrees=30, swing2LimitDegrees=30, twistLimitDegrees=15}
 {bone1Name=dyn_r_bk_hair_03, bone2Name=dyn_r_bk_hair_02, swing1LimitDegrees=30, swing2LimitDegrees=30, twistLimitDegrees=15}
 {
  bone1Name=upperarm_l
  bone2Name=clavicle_l
  swing1LimitDegrees=45
  swing1Motion=Free
  swing2LimitDegrees=45
  swing2Motion=Free
  twistLimitDegrees=45
  twistMotion=Free
 }
 {
  bone1Name=upperarm_r
  bone2Name=clavicle_r
  swing1LimitDegrees=45
  swing1Motion=Free
  swing2LimitDegrees=45
  swing2Motion=Free
  twistLimitDegrees=45
  twistMotion=Free
 }
 {
  bone1Name=clavicle_l
  bone2Name=spine_05
  swing1LimitDegrees=45
  swing1Motion=Free
  swing2LimitDegrees=45
  swing2Motion=Free
  twistLimitDegrees=45
  twistMotion=Free
 }
 {
  bone1Name=clavicle_r
  bone2Name=spine_05
  swing1LimitDegrees=45
  swing1Motion=Free
  swing2LimitDegrees=45
  swing2Motion=Free
  twistLimitDegrees=45
  twistMotion=Free
 }
 {
  bone1Name=dyn_r_bk_hairSide_MSTR
  bone2Name=head
  swing1LimitDegrees=45
  swing1Motion=Free
  swing2LimitDegrees=45
  swing2Motion=Free
  twistLimitDegrees=45
  twistMotion=Free
 }
 {
  bone1Name=deform_upperarm_up_base_r
  bone2Name=upperarm_r
  swing1LimitDegrees=45
  swing1Motion=Free
  swing2LimitDegrees=45
  swing2Motion=Free
  twistLimitDegrees=45
  twistMotion=Free
 }
]
