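# Post-process-style Anim Blueprint for the female Gloom hat and face-accessory part, adding secondary motion to hair, ears, nose ring and hat over the copied parent pose.
# Back hair uses AnimDynamics with body constraints while skydiving, otherwise RigidBody simulation varying with backbling.
# AnimDynamics then moves the smaller pieces, the hat shifts for skydiving, and a crouch, backbling or default Control Rig finishes.
# The event graph updates state, gravity and hair-velocity values.
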AnimBlueprint /BRCosmetics/Characters/Player/Female/Medium/Bodies/F_MED_Gloom/Meshes/Parts/F_MED_Gloom_Hat_FaceAcc_AnimBP
target_skeleton=&/F_MED_Gloom_Hat_FaceAcc_Skeleton
graphs [
 AnimGraph {nodes [
   n1 = "Constraint - dyn_r_bk_hairSide_02 (PPPPPPPP)"!(
    ComponentPose
    Alpha=1
    ConstraintWeights_0=0.0137
    ConstraintWeights_1:float=0.1677
    ConstraintWeights_2:float=0.6591
    ConstraintWeights_3:float=0.0018
    ConstraintWeights_4:float=0.0222
    ConstraintWeights_5:float=0.0549
    ConstraintWeights_6:float=0.0765
    ConstraintWeights_7:float=0.0042
   ) -> (Pose) {
    alphaBoolBlend {blendOption=Linear}
    boneToModify {boneName=dyn_r_bk_hairSide_02}
    constraintSetup [
     {targetBone {boneName=spine_04}}
     {targetBone {boneName=spine_05}}
     {targetBone {boneName=clavicle_r}}
     {targetBone {boneName=deform_upperarm_bk_r}}
     {targetBone {boneName=deform_upperarm_fr_base_r}}
     {targetBone {boneName=deform_upperarm_up_base_r}}
     {targetBone {boneName=upperarm_twist_02_r}}
     {targetBone {boneName=deform_pec_base_r}}
    ]
    constraintWeights [1,1,1,1,1,1,1,1]
   }
   "Constraint - dyn_l_bk_hairSide_02 (PPPPPPPPPP)"!(
    ComponentPose=@n1
    Alpha=1
    ConstraintWeights_0=0
    ConstraintWeights_1:float=0
    ConstraintWeights_2:float=0.0184
    ConstraintWeights_3:float=0.1414
    ConstraintWeights_4:float=0.6879
    ConstraintWeights_5:float=0.0034
    ConstraintWeights_6:float=0.022
    ConstraintWeights_7:float=0.0441
    ConstraintWeights_8:float=0.0767
    ConstraintWeights_9:float=0.0061
   ) -> (Pose) {
    alphaBoolBlend {blendOption=Linear}
    boneToModify {boneName=dyn_l_bk_hairSide_02}
    constraintSetup [
     {targetBone {boneName=spine_02}}
     {targetBone {boneName=spine_03}}
     {targetBone {boneName=spine_04}}
     {targetBone {boneName=spine_05}}
     {targetBone {boneName=clavicle_l}}
     {targetBone {boneName=deform_upperarm_bk_base_l}}
     {targetBone {boneName=deform_upperarm_fr_l}}
     {targetBone {boneName=deform_upperarm_up_base_l}}
     {targetBone {boneName=upperarm_twist_02_l}}
     {targetBone {boneName=deform_pec_l}}
    ]
    constraintWeights [1,1,1,1,1,1,1,1,1,1]
   }
   RigidBody!(
    ComponentPose
    bAlphaBoolEnabled:bool=Get IsWearingBackbling() -> (IsWearingBackbling:bool)
    OverrideWorldGravity:Vector="0,0,-1500"
    ExternalForce
    ComponentLinearAccScale:Vector="0,0,0"
   ) -> (Pose) {
    alphaInputType=Bool
    bClampLinearTranslationLimitToRefPose=true
    bOverrideWorldGravity=true
    overridePhysicsAsset=&/F_MED_Gloom_Hat_FaceAcc_Physics
    simulationSpace=BaseBoneSpace
    alphaBoolBlend {blendOption=Linear}
    baseBoneRef {boneName=spine_05}
    componentAppliedLinearAccClamp=<3000 3000 3000>
    componentLinearAccScale=<0.4 0.4 0.5>
    overrideWorldGravity {z=-980}
    simSpaceSettings {
     maxAngularAcceleration=1000
     maxAngularVelocity=3
     maxLinearAcceleration=700
     maxLinearVelocity=500
     velocityScaleZ=0
     externalLinearDragV=<0.2 0.2>
    }
   }
   Output Pose(Result=Use cached pose 'PreFinal'()) {layerGroup=None}
   n2 = "Constraint - dyn_r_bk_hair_MSTR (P)"(ComponentPose=Local To Component(LocalPose=Copy Pose From Mesh() {bUseAttachedParent=true}), Alpha=1, ConstraintWeights_0=1) {
    alphaBoolBlend {blendOption=Linear}
    boneToModify {boneName=dyn_r_bk_hair_MSTR}
    constraintSetup [{targetBone {boneName=spine_05}}]
    constraintWeights [1]
   }
   n3 = "Constraint - dyn_l_bk_hair_MSTR (P)"(ComponentPose=@n2, Alpha=1, ConstraintWeights_0=1) {
    alphaBoolBlend {blendOption=Linear}
    boneToModify {boneName=dyn_l_bk_hair_MSTR}
    constraintSetup [{targetBone {boneName=spine_05}}]
    constraintWeights [1]
   }
   n4 = "Constraint - dyn_bk_hair_MSTR (P)"(ComponentPose=@n3, Alpha=1, ConstraintWeights_0=1) {
    alphaBoolBlend {blendOption=Linear}
    boneToModify {boneName=dyn_bk_hair_MSTR}
    constraintSetup [{targetBone {boneName=spine_05}}]
    constraintWeights [1]
   }
   n5 = Look At - Bone: dyn_l_bk_hair_MSTR(ComponentPose=@n4, Alpha=1) {
    alphaBoolBlend {blendOption=Linear}
    boneToModify {boneName=dyn_l_bk_hair_MSTR}
    lookAtLocation=<0 -10>
    lookAtTarget {boneReference {boneName=pelvis}}
    lookAt_Axis {axis=<1 0>}
   }
   n6 = Look At - Bone: dyn_bk_hair_MSTR(ComponentPose=@n5, Alpha=1) {
    alphaBoolBlend {blendOption=Linear}
    boneToModify {boneName=dyn_bk_hair_MSTR}
    lookAtLocation=<0 -10>
    lookAtTarget {boneReference {boneName=pelvis}}
    lookAt_Axis {axis=<1 0>}
   }
   n7 = Look At - Bone: dyn_r_bk_hair_MSTR(ComponentPose=@n6, Alpha=1) {
    alphaBoolBlend {blendOption=Linear}
    boneToModify {boneName=dyn_r_bk_hair_MSTR}
    lookAtLocation=<0 -10>
    lookAtTarget {boneReference {boneName=pelvis}}
    lookAt_Axis {axis=<1 0>}
   }
   n8 = "Transform (Modify) Bone - Bone: dyn_r_bk_hairSide_MSTR"(ComponentPose=@n7, Alpha=1, Translation="0,0,0", Rotation="0,50,30", Scale="1,1,1") {
    rotationMode=BMM_Additive
    rotationSpace=BCS_BoneSpace
    scaleMode=BMM_Additive
    scaleSpace=BCS_BoneSpace
    translationMode=BMM_Additive
    translationSpace=BCS_BoneSpace
    alphaBoolBlend {blendOption=Linear}
    boneToModify {boneName=dyn_r_bk_hairSide_MSTR}
    rotation {roll=30, yaw=50}
   }
   n9 = "Constraint - dyn_l_bk_hairSide_02 (P)"(ComponentPose=@n8, Alpha=1, ConstraintWeights_0=1) {
    alphaBoolBlend {blendOption=Linear}
    boneToModify {boneName=dyn_l_bk_hairSide_02}
    constraintSetup [{targetBone {boneName=clavicle_l}}]
    constraintWeights [1]
   }
   n10 = "Constraint - dyn_r_bk_hairSide_02 (P)"(ComponentPose=@n9, Alpha=1, ConstraintWeights_0=1) {
    alphaBoolBlend {blendOption=Linear}
    boneToModify {boneName=dyn_r_bk_hairSide_02}
    constraintSetup [{targetBone {boneName=clavicle_r}}]
    constraintWeights [1]
   }
   Save cached pose 'PreSim'(Pose=Component To Local(ComponentPose=@n10)) {cacheName=PreSim}
   n11 = RigidBody!(
    ComponentPose=Local To Component(LocalPose=Use cached pose 'PreSim'())
    Alpha:float
    OverrideWorldGravity="0,0,-1500"
    ExternalForce
    ComponentLinearAccScale="0,0,0"
   ) -> (Pose) {
    bClampLinearTranslationLimitToRefPose=true
    bOverrideWorldGravity=true
    overridePhysicsAsset=&F_MED_Gloom_Hat_FaceAcc_Physics
    simulationSpace=BaseBoneSpace
    alphaBoolBlend {blendOption=Linear}
    baseBoneRef {boneName=spine_05}
    componentAppliedLinearAccClamp=<3000 3000 3000>
    componentLinearAccScale=<0.4 0.4 0.5>
    overrideWorldGravity {z=-980}
    simSpaceSettings {
     maxAngularAcceleration=1000
     maxAngularVelocity=3
     maxLinearAcceleration=700
     maxLinearVelocity=500
     velocityScaleZ=0
     externalLinearDragV=<0.2 0.2>
    }
   }
   n12 = RigidBody(
    ComponentPose=Local To Component(LocalPose=Use cached pose 'PreSim'())
    OverrideWorldGravity="0,0,-980"
    ComponentLinearAccScale="0.4,0.4,0.5"
   ) {
    bClampLinearTranslationLimitToRefPose=true
    bOverrideWorldGravity=true
    overridePhysicsAsset=&F_MED_Gloom_Hat_FaceAcc_Physics
    simulationSpace=BaseBoneSpace
    alphaBoolBlend {blendOption=Linear}
    baseBoneRef {boneName=spine_05}
    componentAppliedLinearAccClamp=<3000 3000 3000>
    componentLinearAccScale=<0.4 0.4 0.5>
    componentLinearVelScale=<0.3 0.3 0.4>
    overrideWorldGravity {z=-980}
    simSpaceSettings {
     maxAngularAcceleration=1000
     maxAngularVelocity=3
     maxLinearAcceleration=700
     maxLinearVelocity=500
     velocityScaleZ=0
     externalLinearDragV=<0.2 0.2>
    }
   }
   n13 = Blend Poses by bool(
    BlendPose_0=Component To Local(ComponentPose=@n11)
    BlendPose_1=Component To Local(ComponentPose=@n12)
    bActiveValue=Get IsWearingBackbling()
   ) {blendType=Linear}
   n14 = Anim Dynamics - Bone: dyn_r_bk_hair_01!(ComponentPose=Local To Component(LocalPose=Use cached pose 'PreSim'()), Alpha=0.25, GravityOverride:Vector="0,0,0", ExternalForce="0,0,0") -> (Pose) {
    angularDampingOverride=0.9999
    bAngularSpring=true
    bOverrideAngularDamping=true
    bUseGravityOverride=true
    simulationSpace=RootRelative
    alphaBoolBlend {blendInTime=0.1, blendOption=Linear, blendOutTime=0.1}
    boundBone {boneName=dyn_r_bk_hair_01}
    componentAppliedLinearAccClamp=<5000 5000 5000>
    componentLinearAccScale=<2 2 2>
    componentLinearVelScale=<2 2 2>
    physicsBodyDefinitions [{
      sphereCollisionRadius=0
      boundBone {boneName=dyn_r_bk_hair_01}
      boxExtents=<20 20 10>
      constraintSetup {angularLimitsMax=<45 -20 5>, angularLimitsMin=<-45 -45 -5>, angularTarget=<0 0 0>}
      localJointOffset=<1 0 -0.5>
     }]
   }
   n15 = Anim Dynamics - Bone: dyn_l_bk_hair_01!(ComponentPose=@n14, Alpha=0.25, GravityOverride:Vector="0,0,0", ExternalForce="0,0,0") -> (Pose) {
    angularDampingOverride=0.9999
    bAngularSpring=true
    bOverrideAngularDamping=true
    bUseGravityOverride=true
    simulationSpace=RootRelative
    alphaBoolBlend {blendInTime=0.1, blendOption=Linear, blendOutTime=0.1}
    boundBone {boneName=dyn_l_bk_hair_01}
    componentAppliedLinearAccClamp=<5000 5000 5000>
    componentLinearAccScale=<2 2 2>
    componentLinearVelScale=<2 2 2>
    physicsBodyDefinitions [{
      sphereCollisionRadius=0
      boundBone {boneName=dyn_l_bk_hair_01}
      boxExtents=<20 20 10>
      constraintSetup {angularLimitsMax=<45 -20 5>, angularLimitsMin=<-45 -45 -5>, angularTarget=<0 0 0>}
      localJointOffset=<1 0 -0.5>
     }]
   }
   n16 = Anim Dynamics - Bone: dyn_bk_hair_01!(ComponentPose=@n15, Alpha=0.25, GravityOverride:Vector="0,0,0", ExternalForce="0,0,0") -> (Pose) {
    angularDampingOverride=0.9999
    bAngularSpring=true
    bOverrideAngularDamping=true
    bUseGravityOverride=true
    simulationSpace=RootRelative
    alphaBoolBlend {blendInTime=0.1, blendOption=Linear, blendOutTime=0.1}
    boundBone {boneName=dyn_bk_hair_01}
    componentAppliedLinearAccClamp=<5000 5000 5000>
    componentLinearAccScale=<2 2 2>
    componentLinearVelScale=<2 2 2>
    physicsBodyDefinitions [{
      sphereCollisionRadius=0
      boundBone {boneName=dyn_bk_hair_01}
      boxExtents=<20 20 10>
      constraintSetup {angularLimitsMax=<45 -20 5>, angularLimitsMin=<-45 -45 -5>, angularTarget=<0 0 0>}
      localJointOffset=<1 0 -0.5>
     }]
   }
   n17 = Anim Dynamics - Bone: dyn_r_bk_hair_02!(ComponentPose=@n16, Alpha=0.3, GravityOverride:Vector="0,0,0", ExternalForce="0,0,0") -> (Pose) {
    angularDampingOverride=0.9999
    bAngularSpring=true
    bOverrideAngularDamping=true
    bUseGravityOverride=true
    simulationSpace=RootRelative
    alphaBoolBlend {blendInTime=0.1, blendOption=Linear, blendOutTime=0.1}
    boundBone {boneName=dyn_r_bk_hair_02}
    componentAppliedLinearAccClamp=<5000 5000 5000>
    componentLinearAccScale=<2 2 2>
    componentLinearVelScale=<2 2 2>
    physicsBodyDefinitions [{
      sphereCollisionRadius=0
      boundBone {boneName=dyn_r_bk_hair_02}
      boxExtents=<20 20 10>
      constraintSetup {angularLimitsMax=<45 -20 5>, angularLimitsMin=<-45 -20 -5>, angularTarget=<0 0 0>}
      localJointOffset=<1 0 -0.5>
     }]
   }
   n18 = Anim Dynamics - Bone: dyn_l_bk_hair_02!(ComponentPose=@n17, Alpha=0.3, GravityOverride:Vector="0,0,0", ExternalForce="0,0,0") -> (Pose) {
    angularDampingOverride=0.9999
    bAngularSpring=true
    bOverrideAngularDamping=true
    bUseGravityOverride=true
    simulationSpace=RootRelative
    alphaBoolBlend {blendInTime=0.1, blendOption=Linear, blendOutTime=0.1}
    boundBone {boneName=dyn_l_bk_hair_02}
    componentAppliedLinearAccClamp=<5000 5000 5000>
    componentLinearAccScale=<2 2 2>
    componentLinearVelScale=<2 2 2>
    physicsBodyDefinitions [{
      sphereCollisionRadius=0
      boundBone {boneName=dyn_l_bk_hair_02}
      boxExtents=<20 20 10>
      constraintSetup {angularLimitsMax=<45 -20 5>, angularLimitsMin=<-45 -20 -5>, angularTarget=<0 0 0>}
      localJointOffset=<1 0 -0.5>
     }]
   }
   n19 = Anim Dynamics - Bone: dyn_bk_hair_02!(ComponentPose=@n18, Alpha=0.3, GravityOverride:Vector="0,0,0", ExternalForce="0,0,0") -> (Pose) {
    angularDampingOverride=0.9999
    bAngularSpring=true
    bOverrideAngularDamping=true
    bUseGravityOverride=true
    simulationSpace=RootRelative
    alphaBoolBlend {blendInTime=0.1, blendOption=Linear, blendOutTime=0.1}
    boundBone {boneName=dyn_bk_hair_02}
    componentAppliedLinearAccClamp=<5000 5000 5000>
    componentLinearAccScale=<2 2 2>
    componentLinearVelScale=<2 2 2>
    physicsBodyDefinitions [{
      sphereCollisionRadius=0
      boundBone {boneName=dyn_bk_hair_02}
      boxExtents=<20 20 10>
      constraintSetup {angularLimitsMax=<45 -20 5>, angularLimitsMin=<-45 -20 -5>, angularTarget=<0 0 0>}
      localJointOffset=<1 0 -0.5>
     }]
   }
   n20 = Anim Dynamics - Bone: dyn_r_bk_hair_03!(ComponentPose=@n19, Alpha=1, GravityOverride:Vector="0,0,0", ExternalForce="0,0,0") -> (Pose) {
    angularDampingOverride=0.9999
    bAngularSpring=true
    bOverrideAngularDamping=true
    bUseGravityOverride=true
    simulationSpace=RootRelative
    alphaBoolBlend {blendInTime=0.1, blendOption=Linear, blendOutTime=0.1}
    boundBone {boneName=dyn_r_bk_hair_03}
    componentAppliedLinearAccClamp=<5000 5000 5000>
    componentLinearAccScale=<2 2 2>
    componentLinearVelScale=<2 2 2>
    physicsBodyDefinitions [{
      sphereCollisionRadius=0
      boundBone {boneName=dyn_r_bk_hair_03}
      boxExtents=<20 20 10>
      constraintSetup {angularLimitsMax=<45 -20 5>, angularLimitsMin=<-45 -60 -5>, angularTarget=<0 0 0>}
      localJointOffset=<1 0 -0.5>
     }]
   }
   n21 = Anim Dynamics - Bone: dyn_l_bk_hair_03!(ComponentPose=@n20, Alpha=1, GravityOverride:Vector="0,0,0", ExternalForce="0,0,0") -> (Pose) {
    angularDampingOverride=0.9999
    bAngularSpring=true
    bOverrideAngularDamping=true
    bUseGravityOverride=true
    simulationSpace=RootRelative
    alphaBoolBlend {blendInTime=0.1, blendOption=Linear, blendOutTime=0.1}
    boundBone {boneName=dyn_l_bk_hair_03}
    componentAppliedLinearAccClamp=<5000 5000 5000>
    componentLinearAccScale=<2 2 2>
    componentLinearVelScale=<2 2 2>
    physicsBodyDefinitions [{
      sphereCollisionRadius=0
      boundBone {boneName=dyn_l_bk_hair_03}
      boxExtents=<20 20 10>
      constraintSetup {angularLimitsMax=<45 -20 5>, angularLimitsMin=<-45 -60 -5>, angularTarget=<0 0 0>}
      localJointOffset=<1 0 -0.5>
     }]
   }
   n22 = Anim Dynamics - Bone: dyn_bk_hair_03!(ComponentPose=@n21, Alpha=1, GravityOverride:Vector="0,0,0", ExternalForce="0,0,0") -> (Pose) {
    angularDampingOverride=0.9999
    bAngularSpring=true
    bOverrideAngularDamping=true
    bUseGravityOverride=true
    simulationSpace=RootRelative
    alphaBoolBlend {blendInTime=0.1, blendOption=Linear, blendOutTime=0.1}
    boundBone {boneName=dyn_bk_hair_03}
    componentAppliedLinearAccClamp=<5000 5000 5000>
    componentLinearAccScale=<2 2 2>
    componentLinearVelScale=<2 2 2>
    physicsBodyDefinitions [{
      sphereCollisionRadius=0
      boundBone {boneName=dyn_bk_hair_03}
      boxExtents=<20 20 10>
      constraintSetup {angularLimitsMax=<45 -20 5>, angularLimitsMin=<-45 -60 -5>, angularTarget=<0 0 0>}
      localJointOffset=<1 0 -0.5>
     }]
   }
   n23 = "Constraint - dyn_l_bk_hair_02 (PPPPPPPPPPPPP)"!(
    ComponentPose=@n22
    bAlphaBoolEnabled:bool=True
    ConstraintWeights_0=0
    ConstraintWeights_1:float=0
    ConstraintWeights_2:float=0.0038
    ConstraintWeights_3:float=0.1562
    ConstraintWeights_4:float=0
    ConstraintWeights_5:float=0
    ConstraintWeights_6:float=0.6843
    ConstraintWeights_7:float=0.0359
    ConstraintWeights_8:float=0.0026
    ConstraintWeights_9:float=0.0487
    ConstraintWeights_10:float=0.068
    ConstraintWeights_11:float=0
    ConstraintWeights_12:float=0.0005
   ) -> (Pose) {
    alphaInputType=Bool
    alphaBoolBlend {blendOption=Linear}
    boneToModify {boneName=dyn_l_bk_hair_02}
    constraintSetup [
     {targetBone {boneName=spine_02}}
     {targetBone {boneName=spine_03}}
     {targetBone {boneName=spine_04}}
     {targetBone {boneName=spine_05}}
     {targetBone {boneName=neck_01}}
     {targetBone {boneName=neck_02}}
     {targetBone {boneName=clavicle_l}}
     {targetBone {boneName=deform_upperarm_bk_base_l}}
     {targetBone {boneName=deform_upperarm_fr_l}}
     {targetBone {boneName=deform_upperarm_up_base_l}}
     {targetBone {boneName=upperarm_twist_02_l}}
     {targetBone {boneName=clavicle_r}}
     {targetBone {boneName=deform_pec_l}}
    ]
    constraintWeights [1,1,1,1,1,1,1,1,1,1,1,1,1]
   }
   n24 = "Constraint - dyn_bk_hair_02 (PPPPP)"!(
    ComponentPose=@n23
    bAlphaBoolEnabled:bool=True
    ConstraintWeights_0=0.0071
    ConstraintWeights_1:float=0.8565
    ConstraintWeights_2:float=0.0048
    ConstraintWeights_3:float=0.0744
    ConstraintWeights_4:float=0.0572
   ) -> (Pose) {
    alphaInputType=Bool
    alphaBoolBlend {blendOption=Linear}
    boneToModify {boneName=dyn_bk_hair_02}
    constraintSetup [
     {targetBone {boneName=spine_03}}
     {targetBone {boneName=spine_05}}
     {targetBone {boneName=neck_01}}
     {targetBone {boneName=clavicle_l}}
     {targetBone {boneName=clavicle_r}}
    ]
    constraintWeights [1,1,1,1,1]
   }
   n25 = "Constraint - dyn_r_bk_hair_02 (PPPPPPPP)"!(
    ComponentPose=@n24
    bAlphaBoolEnabled:bool=True
    ConstraintWeights_0=0.0022
    ConstraintWeights_1:float=0.1825
    ConstraintWeights_2:float=0.6499
    ConstraintWeights_3:float=0.0395
    ConstraintWeights_4:float=0.0015
    ConstraintWeights_5:float=0.0538
    ConstraintWeights_6:float=0.0703
    ConstraintWeights_7:float=0.0003
   ) -> (Pose) {
    alphaInputType=Bool
    alphaBoolBlend {blendOption=Linear}
    boneToModify {boneName=dyn_r_bk_hair_02}
    constraintSetup [
     {targetBone {boneName=spine_04}}
     {targetBone {boneName=spine_05}}
     {targetBone {boneName=clavicle_r}}
     {targetBone {boneName=deform_upperarm_bk_r}}
     {targetBone {boneName=deform_upperarm_fr_base_r}}
     {targetBone {boneName=deform_upperarm_up_base_r}}
     {targetBone {boneName=upperarm_twist_02_r}}
     {targetBone {boneName=deform_pec_base_r}}
    ]
    constraintWeights [1,1,1,1,1,1,1,1]
   }
   n26 = "Constraint - dyn_l_bk_hair_03 (PPPPPPPPPPPPP)"!(
    ComponentPose=@n25
    bAlphaBoolEnabled:bool=True
    ConstraintWeights_0=0
    ConstraintWeights_1:float=0
    ConstraintWeights_2:float=0.0038
    ConstraintWeights_3:float=0.1562
    ConstraintWeights_4:float=0
    ConstraintWeights_5:float=0
    ConstraintWeights_6:float=0.6843
    ConstraintWeights_7:float=0.0359
    ConstraintWeights_8:float=0.0026
    ConstraintWeights_9:float=0.0487
    ConstraintWeights_10:float=0.068
    ConstraintWeights_11:float=0
    ConstraintWeights_12:float=0.0005
   ) -> (Pose) {
    alphaInputType=Bool
    alphaBoolBlend {blendOption=Linear}
    boneToModify {boneName=dyn_l_bk_hair_03}
    constraintSetup [
     {targetBone {boneName=spine_02}}
     {targetBone {boneName=spine_03}}
     {targetBone {boneName=spine_04}}
     {targetBone {boneName=spine_05}}
     {targetBone {boneName=neck_01}}
     {targetBone {boneName=neck_02}}
     {targetBone {boneName=clavicle_l}}
     {targetBone {boneName=deform_upperarm_bk_base_l}}
     {targetBone {boneName=deform_upperarm_fr_l}}
     {targetBone {boneName=deform_upperarm_up_base_l}}
     {targetBone {boneName=upperarm_twist_02_l}}
     {targetBone {boneName=clavicle_r}}
     {targetBone {boneName=deform_pec_l}}
    ]
    constraintWeights [1,1,1,1,1,1,1,1,1,1,1,1,1]
   }
   n27 = "Constraint - dyn_bk_hair_03 (PPPPP)"!(
    ComponentPose=@n26
    bAlphaBoolEnabled:bool=True
    ConstraintWeights_0=0.0071
    ConstraintWeights_1:float=0.8565
    ConstraintWeights_2:float=0.0048
    ConstraintWeights_3:float=0.0744
    ConstraintWeights_4:float=0.0572
   ) -> (Pose) {
    alphaInputType=Bool
    alphaBoolBlend {blendOption=Linear}
    boneToModify {boneName=dyn_bk_hair_03}
    constraintSetup [
     {targetBone {boneName=spine_03}}
     {targetBone {boneName=spine_05}}
     {targetBone {boneName=neck_01}}
     {targetBone {boneName=clavicle_l}}
     {targetBone {boneName=clavicle_r}}
    ]
    constraintWeights [1,1,1,1,1]
   }
   n28 = "Constraint - dyn_r_bk_hair_03 (PPPPPPPP)"!(
    ComponentPose=@n27
    bAlphaBoolEnabled:bool=True
    ConstraintWeights_0=0.0022
    ConstraintWeights_1:float=0.1825
    ConstraintWeights_2:float=0.6499
    ConstraintWeights_3:float=0.0395
    ConstraintWeights_4:float=0.0015
    ConstraintWeights_5:float=0.0538
    ConstraintWeights_6:float=0.0703
    ConstraintWeights_7:float=0.0003
   ) -> (Pose) {
    alphaInputType=Bool
    alphaBoolBlend {blendOption=Linear}
    boneToModify {boneName=dyn_r_bk_hair_03}
    constraintSetup [
     {targetBone {boneName=spine_04}}
     {targetBone {boneName=spine_05}}
     {targetBone {boneName=clavicle_r}}
     {targetBone {boneName=deform_upperarm_bk_r}}
     {targetBone {boneName=deform_upperarm_fr_base_r}}
     {targetBone {boneName=deform_upperarm_up_base_r}}
     {targetBone {boneName=upperarm_twist_02_r}}
     {targetBone {boneName=deform_pec_base_r}}
    ]
    constraintWeights [1,1,1,1,1,1,1,1]
   }
   n29 = Component To Local(ComponentPose=@n28)
   n30 = Get SkydivingNew() -> (SkydivingNew:bool)
   n31 = Anim Dynamics - Bone: dyn_hairBang_01!(
    ComponentPose=Local To Component(LocalPose=Blend Poses by bool(BlendPose_0=@n29, BlendPose_1=@n13, bActiveValue=@n30) {blendType=Linear})
    Alpha=0.6
    GravityOverride:Vector="0,0,0"
    ExternalForce="0,0,0"
   ) -> (Pose) {
    angularDampingOverride=0.9999
    angularSpringConstant=250
    bAngularSpring=true
    bGravityOverrideInSimSpace=true
    bOverrideAngularDamping=true
    bUseGravityOverride=true
    simulationSpace=RootRelative
    alphaBoolBlend {blendOption=Linear}
    boundBone {boneName=dyn_hairBang_01}
    componentAppliedLinearAccClamp=<5000 5000 5000>
    componentLinearAccScale=<1 1 1>
    componentLinearVelScale=<1 1 1>
    physicsBodyDefinitions [{
      sphereCollisionRadius=0
      boundBone {boneName=dyn_hairBang_01}
      boxExtents=<10 10 5>
      constraintSetup {angularLimitsMax=<15 1.5 15>, angularLimitsMin=<-15 -25 -15>, angularTarget=<0 0 0>}
      localJointOffset=<1 0 -0.5>
     }]
   }
   n32 = Anim Dynamics - Bone: dyn_hairBang_02!(ComponentPose=@n31, Alpha=0.6, GravityOverride:Vector="0,0,0", ExternalForce="0,0,0") -> (Pose) {
    angularDampingOverride=0.9999
    angularSpringConstant=250
    bAngularSpring=true
    bGravityOverrideInSimSpace=true
    bOverrideAngularDamping=true
    bUseGravityOverride=true
    simulationSpace=RootRelative
    alphaBoolBlend {blendOption=Linear}
    boundBone {boneName=dyn_hairBang_02}
    componentAppliedLinearAccClamp=<5000 5000 5000>
    componentLinearAccScale=<1 1 1>
    componentLinearVelScale=<1 1 1>
    physicsBodyDefinitions [{
      sphereCollisionRadius=0
      boundBone {boneName=dyn_hairBang_02}
      boxExtents=<12 12 8>
      constraintSetup {angularLimitsMax=<15 1.5 15>, angularLimitsMin=<-15 -25 -15>, angularTarget=<0 0 0>}
      localJointOffset=<1 0 -0.5>
     }]
   }
   n33 = Anim Dynamics - Bone: dyn_hairBang_03!(ComponentPose=@n32, Alpha=0.6, GravityOverride:Vector="0,0,0", ExternalForce="0,0,0") -> (Pose) {
    angularDampingOverride=0.9999
    angularSpringConstant=250
    bAngularSpring=true
    bGravityOverrideInSimSpace=true
    bOverrideAngularDamping=true
    bUseGravityOverride=true
    simulationSpace=RootRelative
    alphaBoolBlend {blendOption=Linear}
    boundBone {boneName=dyn_hairBang_03}
    componentAppliedLinearAccClamp=<5000 5000 5000>
    componentLinearAccScale=<1 1 1>
    componentLinearVelScale=<1 1 1>
    physicsBodyDefinitions [{
      sphereCollisionRadius=0
      boundBone {boneName=dyn_hairBang_03}
      boxExtents=<10 10 5>
      constraintSetup {angularLimitsMax=<15 -3 15>, angularLimitsMin=<-15 -25 -15>, angularTarget=<0 0 0>}
      localJointOffset=<1 0 -0.5>
     }]
   }
   n34 = Anim Dynamics - Bone: dyn_hairBang_04!(ComponentPose=@n33, Alpha=0.6, GravityOverride:Vector="0,0,0", ExternalForce="0,0,0") -> (Pose) {
    angularDampingOverride=0.9999
    angularSpringConstant=250
    bAngularSpring=true
    bGravityOverrideInSimSpace=true
    bOverrideAngularDamping=true
    bUseGravityOverride=true
    simulationSpace=RootRelative
    alphaBoolBlend {blendOption=Linear}
    boundBone {boneName=dyn_hairBang_04}
    componentAppliedLinearAccClamp=<5000 5000 5000>
    componentLinearAccScale=<1 1 1>
    componentLinearVelScale=<1 1 1>
    physicsBodyDefinitions [{
      sphereCollisionRadius=0
      boundBone {boneName=dyn_hairBang_04}
      boxExtents=<12 12 8>
      constraintSetup {angularLimitsMax=<15 1.5 15>, angularLimitsMin=<-15 -25 -15>, angularTarget=<0 0 0>}
      localJointOffset=<1 0 -0.5>
     }]
   }
   n35 = Anim Dynamics - Bone: dyn_l_sideHair_01!(ComponentPose=@n34, Alpha=1, GravityOverride:Vector="0,0,0", ExternalForce="0,0,0") -> (Pose) {
    angularDampingOverride=0.9999
    angularSpringConstant=250
    bAngularSpring=true
    bGravityOverrideInSimSpace=true
    bOverrideAngularDamping=true
    bUseGravityOverride=true
    simulationSpace=RootRelative
    alphaBoolBlend {blendOption=Linear}
    boundBone {boneName=dyn_l_sideHair_01}
    componentAppliedLinearAccClamp=<5000 5000 5000>
    componentLinearAccScale=<1 1 1>
    componentLinearVelScale=<1 1 1>
    physicsBodyDefinitions [{
      sphereCollisionRadius=0
      boundBone {boneName=dyn_l_sideHair_01}
      boxExtents=<5 10 10>
      constraintSetup {angularLimitsMax=<5 0 45>, angularLimitsMin=<-5 -40 -6>, angularTarget=<0 0 0>}
      localJointOffset=<1 0 0>
     }]
   }
   n36 = Anim Dynamics - Bone: dyn_r_sideHair_01!(ComponentPose=@n35, Alpha=1, GravityOverride:Vector="0,0,0", ExternalForce="0,0,0") -> (Pose) {
    angularDampingOverride=0.9999
    angularSpringConstant=250
    bAngularSpring=true
    bGravityOverrideInSimSpace=true
    bOverrideAngularDamping=true
    bUseGravityOverride=true
    simulationSpace=RootRelative
    alphaBoolBlend {blendOption=Linear}
    boundBone {boneName=dyn_r_sideHair_01}
    componentAppliedLinearAccClamp=<5000 5000 5000>
    componentLinearAccScale=<1 1 1>
    componentLinearVelScale=<1 1 1>
    physicsBodyDefinitions [{
      sphereCollisionRadius=0
      boundBone {boneName=dyn_r_sideHair_01}
      boxExtents=<8 15 15>
      constraintSetup {angularLimitsMax=<0 5 45>, angularLimitsMin=<0 -5 -20>, angularTarget=<0 0 0>}
      localJointOffset=<1 0 0>
     }]
   }
   n37 = Anim Dynamics - Bone: dyn_r_sideHair_02!(ComponentPose=@n36, Alpha=1, GravityOverride:Vector="0,0,0", ExternalForce="0,0,0") -> (Pose) {
    angularDampingOverride=0.9999
    angularSpringConstant=250
    bAngularSpring=true
    bGravityOverrideInSimSpace=true
    bOverrideAngularDamping=true
    bUseGravityOverride=true
    simulationSpace=RootRelative
    alphaBoolBlend {blendOption=Linear}
    boundBone {boneName=dyn_r_sideHair_02}
    componentAppliedLinearAccClamp=<5000 5000 5000>
    componentLinearAccScale=<1 1 1>
    componentLinearVelScale=<1 1 1>
    physicsBodyDefinitions [{
      sphereCollisionRadius=0
      boundBone {boneName=dyn_r_sideHair_02}
      boxExtents=<8 15 15>
      constraintSetup {angularLimitsMax=<5 0 6>, angularLimitsMin=<-5 -40 -45>, angularTarget=<0 0 0>}
      localJointOffset=<1 0 0>
     }]
   }
   n38 = Anim Dynamics - Bone: dyn_l_ear(ComponentPose=@n37, Alpha=1, ExternalForce="0,0,0") {
    angularDampingOverride=0.9999
    bAngularSpring=true
    bGravityOverrideInSimSpace=true
    bOverrideAngularDamping=true
    simulationSpace=RootRelative
    alphaBoolBlend {blendOption=Linear}
    boundBone {boneName=dyn_l_ear}
    componentLinearAccScale=<2 2 2>
    componentLinearVelScale=<2 2 2>
    physicsBodyDefinitions [{
      sphereCollisionRadius=0
      boundBone {boneName=dyn_l_ear}
      boxExtents=<15 15 8>
      constraintSetup {angularLimitsMax=<30 30 45>, angularLimitsMin=<-30 -30 -45>, angularTarget=<0 0 0>}
      localJointOffset=<2 0 0>
     }]
   }
   n39 = Anim Dynamics - Bone: dyn_r_ear(ComponentPose=@n38, Alpha=1, ExternalForce="0,0,0") {
    angularDampingOverride=0.9999
    bAngularSpring=true
    bGravityOverrideInSimSpace=true
    bOverrideAngularDamping=true
    simulationSpace=RootRelative
    alphaBoolBlend {blendOption=Linear}
    boundBone {boneName=dyn_r_ear}
    componentLinearAccScale=<2 2 2>
    componentLinearVelScale=<2 2 2>
    physicsBodyDefinitions [{
      sphereCollisionRadius=0
      boundBone {boneName=dyn_r_ear}
      boxExtents=<15 15 8>
      constraintSetup {angularLimitsMax=<30 30 45>, angularLimitsMin=<-30 -30 -45>, angularTarget=<0 0 0>}
      localJointOffset=<2 0 0>
     }]
   }
   n40 = "Constraint - dyn_noseRing_MSTR (P)"(ComponentPose=@n39, Alpha=1, ConstraintWeights_0=1) {
    alphaBoolBlend {blendOption=Linear}
    boneToModify {boneName=dyn_noseRing_MSTR}
    constraintSetup [{targetBone {boneName=head}}]
    constraintWeights [1]
   }
   n41 = Anim Dynamics - Bone: dyn_noseRing(ComponentPose=@n40, Alpha=0.7, ExternalForce="0,0,0") {
    angularDampingOverride=0.9999
    bAngularSpring=true
    bGravityOverrideInSimSpace=true
    bOverrideAngularDamping=true
    simulationSpace=RootRelative
    alphaBoolBlend {blendOption=Linear}
    boundBone {boneName=dyn_noseRing}
    componentLinearAccScale=<2 2 2>
    componentLinearVelScale=<2 2 2>
    physicsBodyDefinitions [{
      sphereCollisionRadius=0
      boundBone {boneName=dyn_noseRing}
      boxExtents=<15 15 8>
      constraintSetup {angularLimitsMax=<30 30 45>, angularLimitsMin=<-30 -30 -45>, angularTarget=<0 0 0>}
      localJointOffset=<2 0 0>
     }]
   }
   n42 = "Constraint - dyn_noseRing (P)"!(ComponentPose=@n41, bAlphaBoolEnabled:bool=True, ConstraintWeights_0=1) -> (Pose) {
    alphaInputType=Bool
    alphaBoolBlend {blendOption=Linear}
    boneToModify {boneName=dyn_noseRing}
    constraintSetup [{targetBone {boneName=head}}]
    constraintWeights [1]
   }
   n43 = "Transform (Modify) Bone - Bone: dyn_hat_MSTR"!(ComponentPose=@n42, bAlphaBoolEnabled:bool=Get bIsSkyDiving(), Translation="0,0,2", Rotation="-15,0,0", Scale="1,1,1.1") -> (Pose) {
    alphaInputType=Bool
    rotationMode=BMM_Additive
    rotationSpace=BCS_BoneSpace
    scaleMode=BMM_Additive
    scaleSpace=BCS_BoneSpace
    translationMode=BMM_Additive
    translationSpace=BCS_BoneSpace
    alphaBoolBlend {blendInTime=0.1, blendOption=Linear, blendOutTime=0.1}
    boneToModify {boneName=dyn_hat_MSTR}
    rotation {pitch=-15}
    scale {z=1.1}
    translation {z=2}
   }
   Save cached pose 'PostSim'(Pose=Component To Local(ComponentPose=@n43)) {cacheName=PostSim}
   n44 = Control Rig(Source=Use cached pose 'PostSim'()) {
    alphaBoolBlend {blendOption=Linear}
    controlRigAssetReference {blueprintRigClass=&/F_MED_Gloom_Hat_FaceAcc_CtrlRig.F_MED_Gloom_Hat_FaceAcc_CtrlRig_C}
   }
   n45 = Control Rig(Source=Use cached pose 'PostSim'()) {
    alphaBoolBlend {blendOption=Linear}
    controlRigAssetReference {blueprintRigClass=&/F_MED_Gloom_Hat_Backpack_FaceAcc_CtrlRig.F_MED_Gloom_Hat_Backpack_FaceAcc_CtrlRig_C}
   }
   n46 = Get IsWearingBackbling()
   n47 = Control Rig(Source=Use cached pose 'PostSim'()) {
    alphaBoolBlend {blendOption=Linear}
    controlRigAssetReference {blueprintRigClass=&/F_MED_Gloom_Hat_Crouching_FaceAcc_CtrlRig.F_MED_Gloom_Hat_Crouching_FaceAcc_CtrlRig_C}
   }
   Save cached pose 'PreFinal'(Pose=Blend Poses by bool(
     BlendPose_0=@n47
     BlendPose_1=Blend Poses by bool(BlendPose_0=@n45, BlendPose_1=@n44, bActiveValue=@n46) {blendType=Linear}
     bActiveValue=Get bIsCrouching()
    ) {blendType=Linear}) {cacheName=PreFinal}
  ]}
 EventGraph {nodes [
   Event BlueprintUpdateAnimation()
   > n1 = Sequence!(execute) -> (then_0, then_1, then_2:Exec, then_3:Exec)
   n2 = Get isWearingCape() -> (isWearingCape:bool)
   n3 = Get bIsTargeting()
   n4 = Get bIsPlayingMeleeAnim()
   n5 = Get bIsSkyDiving()
   Set ADS Hair Velocity(
    execute=@n1.then_1
    ADS Hair Velocity:Vector=SelectVector(A="0.15,0.15,0.25", B="0.35,0.6,0.35", bPickA=OR Boolean!(A=@n3, B=@n4, C:bool=@n5, D:bool=@n2) -> (ReturnValue))
   ) -> (Output_Get:Vector)
   n6 = BreakTransform(InTransform=GetDeltaTransformFromRefPose(self=GetOwnerSkeletalMeshComponent(), BoneName=upperarm_r, BaseName=clavicle_r))
   > Set AimUp(AimUp:double=MapRangeClamped(Value=BreakRotator(InRot=@n6.Rotation).Yaw, InRangeA=90, InRangeB=110)) -> (Output_Get:double)
   n7 = Branch(execute=@n1.then_2, Condition=Get bIsSkyDiving())
   > Set SkydivingNew(SkydivingNew:bool=true) -> (Output_Get:bool)
   Set SkydivingNew(execute=@n7.else, SkydivingNew=false)
   Set BackblingNew(execute=@n1.then_3, BackblingNew:bool=CheckForBackbling(self=Self-Reference())) -> (Output_Get:bool)
   n8 = Gravity Override State Function DEPRECATED(
    execute=@n1.then_0
    CharacterPartAnimInstance=Self-Reference()
    GameTimeInSeconds=GetGameTimeInSeconds()
    GravityOverrideParams=Get Gravity Override Params()
   )
   > Set GravityOverride(GravityOverride=@n8)
   Event OnCharacterCustomizationLoaded()
   n9 = CheckForBackbling()
   > Set IsWearingBackbling(IsWearingBackbling:bool=@n9) -> (Output_Get:bool)
   > Set isWearingCape(isWearingCape=@n9.bIsBackblingACape)
  ]}
]
variables [
 ADS Hair Velocity {type=Vector}
 AimUp {type=number}
 BackblingNew {type=boolean}
 Gravity Override Params {type=GravityOverrideParamsStruct}
 GravityOverride {type=Vector}
 isNotWearingBackbling {type=number}
 IsWearingBackbling {type=boolean}
 isWearingBackpack {type=number}
 isWearingCape {type=boolean}
 NewVar {type=boolean}
 SkydivingNew {type=boolean}
]
